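# Physics asset for the Dark Shaman medium male body cosmetic, covering its dynamic loincloth and shoulder fur bones.
# Front and back loincloth are five-box chains hanging from the pelvis with tight twist limits, alongside pelvis and thigh capsules that give the body's shape.
# Flat fur boxes attach with narrow limits to tiny shoulder pad capsules.
# Pads and thighs have free angular constraints, mostly tied straight to the pelvis.
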PhysicsAsset /BRCosmetics/Characters/Player/Male/Medium/Bodies/M_MED_DarkShaman/Meshes/M_MED_Dark_Shaman_Physics
bodies {
 dyn_loincloth_back_01 [{shapeType=Box, extentX=8, extentY=2, extentZ=10, center=<4 0 0>}]
 dyn_loincloth_back_02 [{shapeType=Box, extentX=8, extentY=3, extentZ=10, center=<4 0 0>}]
 dyn_loincloth_back_03 [{shapeType=Box, extentX=8, extentY=3, extentZ=10, center=<4 0 0>}]
 dyn_loincloth_back_04 [{shapeType=Box, extentX=8, extentY=3, extentZ=10, center=<4 0 0>}]
 dyn_loincloth_back_05 [{shapeType=Box, extentX=8, extentY=3, extentZ=10, center=<4 0 0>}]
 dyn_loincloth_front_01 [{shapeType=Box, extentX=8, extentY=2, extentZ=10, center=<4 0 0>}]
 dyn_loincloth_front_02 [{shapeType=Box, extentX=8, extentY=2, extentZ=10, center=<4 0 0>}]
 dyn_loincloth_front_03 [{shapeType=Box, extentX=8, extentY=2, extentZ=10, center=<4 0 0>}]
 dyn_loincloth_front_04 [{shapeType=Box, extentX=8, extentY=2, extentZ=10, center=<4 0 0>}]
 dyn_loincloth_front_05 [{shapeType=Box, extentX=8, extentY=2, extentZ=10, center=<4 0 0>}]
 dyn_shoulder_fur_01_l [{shapeType=Box, extentX=10, extentY=10, extentZ=2, center=<2 0 0>}]
 dyn_shoulder_fur_01_r [{shapeType=Box, extentX=10, extentY=10, extentZ=2, center=<2 0 0>}]
 dyn_shoulder_fur_02_l [{shapeType=Box, extentX=10, extentY=10, extentZ=2, center=<2 0 0>}]
 dyn_shoulder_fur_02_r [{shapeType=Box, extentX=10, extentY=10, extentZ=2, center=<2 0 0>}]
 dyn_shoulder_fur_03_l [{shapeType=Box, extentX=10, extentY=10, extentZ=2, center=<2 0 0>}]
 dyn_shoulder_fur_03_r [{shapeType=Box, extentX=10, extentY=10, extentZ=2, center=<2 0 0>}]
 dyn_shoulder_fur_04_l [{shapeType=Box, extentX=10, extentY=10, extentZ=2, center=<2 0 0>}]
 dyn_shoulder_fur_04_r [{shapeType=Box, extentX=10, extentY=10, extentZ=2, center=<2 0 0>}]
 dyn_shoulder_fur_05_l [{shapeType=Box, extentX=10, extentY=10, extentZ=2, center=<2 0 0>}]
 dyn_shoulder_fur_05_r [{shapeType=Box, extentX=10, extentY=10, extentZ=2, center=<2 0 0>}]
 dyn_shoulderpad_back_01_l [{shapeType=Capsule, length=1, radius=1}]
 dyn_shoulderpad_back_01_r [{shapeType=Capsule, length=1, radius=1}]
 dyn_shoulderpad_front_01_l [{shapeType=Capsule, length=1, radius=1}]
 dyn_shoulderpad_front_01_r [{shapeType=Capsule, length=1, radius=1}]
 dyn_shoulderpad_front_02_l [{shapeType=Capsule, length=1, radius=1}]
 dyn_shoulderpad_front_02_r [{shapeType=Capsule, length=1, radius=1}]
 dyn_shoulderpad_mid_02_l [{shapeType=Capsule, length=1, radius=1}]
 dyn_shoulderpad_mid_02_r [{shapeType=Capsule, length=1, radius=1}]
 pelvis [
  {shapeType=Capsule, length=11.5999, radius=11.1176, center=<-2.6407 -0.1128 0>}
  {shapeType=Capsule, length=11.5999, radius=11.1176, center=<-7.3522 0.9518 0>}
 ]
 thigh_l [
  {shapeType=Capsule, length=32, radius=8, center=<-15.0004 -1.2252 -0.0021>, rotation=<rot 90 0 0>}
  {shapeType=Capsule, length=32, radius=8, center=<-15.3911 3.8223 0.777>, rotation=<rot 90 0.0049 0.0049>}
 ]
 thigh_r [
  {shapeType=Capsule, length=32, radius=8, center=<15.0001 1.2252 -0.0004>, rotation=<rot 90 0 0>}
  {shapeType=Capsule, length=32, radius=8, center=<14.9659 -3.8977 -0.697>, rotation=<rot 90 -0.0396 -0.0396>}
 ]
}
constraints [
 {
  bone1Name=thigh_l
  bone2Name=pelvis
  swing1LimitDegrees=45
  swing1Motion=Free
  swing2LimitDegrees=45
  swing2Motion=Free
  twistLimitDegrees=45
  twistMotion=Free
 }
 {
  bone1Name=thigh_r
  bone2Name=pelvis
  swing1LimitDegrees=45
  swing1Motion=Free
  swing2LimitDegrees=45
  swing2Motion=Free
  twistLimitDegrees=45
  twistMotion=Free
 }
 {bone1Name=dyn_loincloth_back_01, bone2Name=pelvis, swing1LimitDegrees=45, swing2LimitDegrees=45, twistLimitDegrees=5}
 {bone1Name=dyn_loincloth_back_02, bone2Name=dyn_loincloth_back_01, swing1LimitDegrees=20, swing2LimitDegrees=20, twistLimitDegrees=5}
 {bone1Name=dyn_loincloth_back_03, bone2Name=dyn_loincloth_back_02, swing1LimitDegrees=45, swing2LimitDegrees=45, twistLimitDegrees=5}
 {bone1Name=dyn_loincloth_back_04, bone2Name=dyn_loincloth_back_03, swing1LimitDegrees=45, swing2LimitDegrees=45, twistLimitDegrees=5}
 {bone1Name=dyn_loincloth_back_05, bone2Name=dyn_loincloth_back_04, swing1LimitDegrees=45, swing2LimitDegrees=45, twistLimitDegrees=5}
 {bone1Name=dyn_loincloth_front_01, bone2Name=pelvis, swing1LimitDegrees=45, swing2LimitDegrees=45, twistLimitDegrees=5}
 {bone1Name=dyn_loincloth_front_02, bone2Name=dyn_loincloth_front_01, swing1LimitDegrees=45, swing2LimitDegrees=45, twistLimitDegrees=5}
 {bone1Name=dyn_loincloth_front_03, bone2Name=dyn_loincloth_front_02, swing1LimitDegrees=45, swing2LimitDegrees=45, twistLimitDegrees=5}
 {bone1Name=dyn_loincloth_front_04, bone2Name=dyn_loincloth_front_03, swing1LimitDegrees=45, swing2LimitDegrees=45, twistLimitDegrees=5}
 {bone1Name=dyn_loincloth_front_05, bone2Name=dyn_loincloth_front_04, swing1LimitDegrees=45, swing2LimitDegrees=45, twistLimitDegrees=5}
 {
  bone1Name=dyn_shoulderpad_back_01_l
  bone2Name=pelvis
  swing1LimitDegrees=45
  swing1Motion=Free
  swing2LimitDegrees=45
  swing2Motion=Free
  twistLimitDegrees=45
  twistMotion=Free
 }
 {
  bone1Name=dyn_shoulderpad_mid_02_l
  bone2Name=pelvis
  swing1LimitDegrees=45
  swing1Motion=Free
  swing2LimitDegrees=45
  swing2Motion=Free
  twistLimitDegrees=45
  twistMotion=Free
 }
 {
  bone1Name=dyn_shoulderpad_front_01_l
  bone2Name=pelvis
  swing1LimitDegrees=45
  swing1Motion=Free
  swing2LimitDegrees=45
  swing2Motion=Free
  twistLimitDegrees=45
  twistMotion=Free
 }
 {
  bone1Name=dyn_shoulderpad_front_02_l
  bone2Name=dyn_shoulderpad_front_01_l
  swing1LimitDegrees=45
  swing1Motion=Free
  swing2LimitDegrees=45
  swing2Motion=Free
  twistLimitDegrees=45
  twistMotion=Free
 }
 {
  bone1Name=dyn_shoulderpad_back_01_r
  bone2Name=pelvis
  swing1LimitDegrees=45
  swing1Motion=Free
  swing2LimitDegrees=45
  swing2Motion=Free
  twistLimitDegrees=45
  twistMotion=Free
 }
 {
  bone1Name=dyn_shoulderpad_mid_02_r
  bone2Name=pelvis
  swing1LimitDegrees=45
  swing1Motion=Free
  swing2LimitDegrees=45
  swing2Motion=Free
  twistLimitDegrees=45
  twistMotion=Free
 }
 {
  bone1Name=dyn_shoulderpad_front_01_r
  bone2Name=pelvis
  swing1LimitDegrees=45
  swing1Motion=Free
  swing2LimitDegrees=45
  swing2Motion=Free
  twistLimitDegrees=45
  twistMotion=Free
 }
 {
  bone1Name=dyn_shoulderpad_front_02_r
  bone2Name=dyn_shoulderpad_front_01_r
  swing1LimitDegrees=45
  swing1Motion=Free
  swing2LimitDegrees=45
  swing2Motion=Free
  twistLimitDegrees=45
  twistMotion=Free
 }
 {bone1Name=dyn_shoulder_fur_01_l, bone2Name=dyn_shoulderpad_back_01_l, swing1LimitDegrees=20, swing2LimitDegrees=20, twistLimitDegrees=2}
 {bone1Name=dyn_shoulder_fur_02_l, bone2Name=dyn_shoulderpad_mid_02_l, swing1LimitDegrees=20, swing2LimitDegrees=20, twistLimitDegrees=2}
 {bone1Name=dyn_shoulder_fur_03_l, bone2Name=dyn_shoulderpad_mid_02_l, swing1LimitDegrees=20, swing2LimitDegrees=20, twistLimitDegrees=2}
 {bone1Name=dyn_shoulder_fur_05_l, bone2Name=dyn_shoulderpad_front_02_l, swing1LimitDegrees=20, swing2LimitDegrees=20, twistLimitDegrees=2}
 {bone1Name=dyn_shoulder_fur_04_l, bone2Name=dyn_shoulderpad_front_01_l, swing1LimitDegrees=20, swing2LimitDegrees=20, twistLimitDegrees=2}
 {bone1Name=dyn_shoulder_fur_01_r, bone2Name=dyn_shoulderpad_back_01_r, swing1LimitDegrees=20, swing2LimitDegrees=20, twistLimitDegrees=2}
 {bone1Name=dyn_shoulder_fur_03_r, bone2Name=dyn_shoulderpad_mid_02_r, swing1LimitDegrees=20, swing2LimitDegrees=20, twistLimitDegrees=2}
 {bone1Name=dyn_shoulder_fur_02_r, bone2Name=dyn_shoulderpad_mid_02_r, swing1LimitDegrees=20, swing2LimitDegrees=20, twistLimitDegrees=2}
 {bone1Name=dyn_shoulder_fur_05_r, bone2Name=dyn_shoulderpad_front_02_r, swing1LimitDegrees=20, swing2LimitDegrees=20, twistLimitDegrees=2}
 {bone1Name=dyn_shoulder_fur_04_r, bone2Name=dyn_shoulderpad_front_01_r, swing1LimitDegrees=20, swing2LimitDegrees=20, twistLimitDegrees=2}
]
