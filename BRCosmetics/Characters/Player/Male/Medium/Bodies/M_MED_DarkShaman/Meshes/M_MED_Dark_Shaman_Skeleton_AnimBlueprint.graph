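# Anim Blueprint for the Dark Shaman player body cosmetic, copying the attached parent mesh's pose and adding secondary motion.
# RBF Pose Drivers set shoulderpad bones from arm and clavicle rotation, switching between skydiving and grounded pose assets on bIsSkyDiving.
# AnimDynamics swings dyn_llama, which gets an additive crouch offset weighted by left thigh yaw computed each update.
# RigidBody, then AnimDynamics on skirt and shoulder fur, receive the skydiving external force.
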AnimBlueprint /BRCosmetics/Characters/Player/Male/Medium/Bodies/M_MED_DarkShaman/Meshes/M_MED_Dark_Shaman_Skeleton_AnimBlueprint
target_skeleton=/Game/Characters/Player/Male/Medium/Bodies/M_MED_DarkShaman/Meshes/M_MED_Dark_Shaman_Skeleton
graphs [
 AnimGraph {nodes [
   n1 = Pose Driver - Source: upperarm_r!(SourcePose:Pose Link Structure=Use cached pose 'CopyPose_Cache'()) -> (Pose:Pose Link Structure) {
    is=title AnimGraphNode_PoseDriver
    axisLength=1
    poseAsset=&/M_MED_Dark_Shaman_Skydiving_RBF_PoseAsset
    onlyDriveBones [
     {boneName=dyn_shoulderpad_top_pivot_r}
     {boneName=dyn_shoulderpad_master_r}
     {boneName=dyn_shoulderpad_bottom_pivot_r}
     {boneName=dyn_shoulderpad_front_pivot_r}
     {boneName=dyn_shoulderpad_back_pivot_r}
     {boneName=dyn_shoulderpad_top_r}
     {boneName=dyn_shoulderpad_back_01_r}
     {boneName=dyn_shoulderpad_back_02_r}
     {boneName=dyn_shoulderpad_mid_01_r}
     {boneName=dyn_shoulderpad_mid_02_r}
     {boneName=dyn_shoulderpad_front_01_r}
     {boneName=dyn_shoulderpad_front_02_r}
    ]
    poseTargets [
     {
      bApplyCustomCurve=false
      bIsHidden=false
      distanceMethod=DefaultMethod
      drivenName=base_pose_skydiving
      functionType=DefaultFunction
      targetScale=0.5851
      boneTransforms [{targetRotation=<rot -64.7854 -3.089 6.0959>, targetTranslation=<-15.5111 0 0>}]
      customCurve {defaultValue=3.4028234663852886e+38, postInfinityExtrap=RCCE_Constant, preInfinityExtrap=RCCE_Constant, keys []}
      targetRotation=<rot 0 0 0>
     }
     {
      bApplyCustomCurve=false
      bIsHidden=false
      distanceMethod=DefaultMethod
      drivenName=arms_out_skydiving
      functionType=DefaultFunction
      targetScale=1
      boneTransforms [{targetRotation=<rot -3.7155 0.3236 -58.8148>, targetTranslation=<-15.5111 0 0>}]
      customCurve {defaultValue=3.4028234663852886e+38, postInfinityExtrap=RCCE_Constant, preInfinityExtrap=RCCE_Constant, keys []}
      targetRotation=<rot 0 0 0>
     }
     {
      bApplyCustomCurve=false
      bIsHidden=false
      distanceMethod=DefaultMethod
      drivenName=arms_up_skydiving
      functionType=DefaultFunction
      targetScale=1
      boneTransforms [{targetRotation=<rot 39.4402 25.0244 -97.0687>, targetTranslation=<-15.5111 0 0>}]
      customCurve {defaultValue=3.4028234663852886e+38, postInfinityExtrap=RCCE_Constant, preInfinityExtrap=RCCE_Constant, keys []}
      targetRotation=<rot 0 0 0>
     }
     {
      bApplyCustomCurve=false
      bIsHidden=false
      distanceMethod=DefaultMethod
      drivenName=arms_back_skydiving
      functionType=DefaultFunction
      targetScale=0.5851
      boneTransforms [{targetRotation=<rot -48.2244 -47.1933 57.9696>, targetTranslation=<-15.5111 0 0>}]
      customCurve {defaultValue=3.4028234663852886e+38, postInfinityExtrap=RCCE_Constant, preInfinityExtrap=RCCE_Constant, keys []}
      targetRotation=<rot 0 0 0>
     }
    ]
    rBFParams {radius=24.3888}
    sourceBones [{boneName=upperarm_r}]
   }
   n2 = Pose Driver - Source: upperarm_l!(SourcePose:Pose Link Structure=@n1) -> (Pose:Pose Link Structure) {
    is=title AnimGraphNode_PoseDriver
    poseAsset=&M_MED_Dark_Shaman_Skydiving_RBF_PoseAsset
    onlyDriveBones [
     {boneName=dyn_shoulderpad_top_pivot_l}
     {boneName=dyn_shoulderpad_master_l}
     {boneName=dyn_shoulderpad_bottom_pivot_l}
     {boneName=dyn_shoulderpad_front_pivot_l}
     {boneName=dyn_shoulderpad_back_pivot_l}
     {boneName=dyn_shoulderpad_top_l}
     {boneName=dyn_shoulderpad_back_01_l}
     {boneName=dyn_shoulderpad_back_02_l}
     {boneName=dyn_shoulderpad_mid_01_l}
     {boneName=dyn_shoulderpad_mid_02_l}
     {boneName=dyn_shoulderpad_front_01_l}
     {boneName=dyn_shoulderpad_front_02_l}
    ]
    poseTargets [
     {
      bApplyCustomCurve=false
      bIsHidden=false
      distanceMethod=DefaultMethod
      drivenName=base_pose_skydiving
      functionType=DefaultFunction
      targetScale=0.6771
      boneTransforms [{targetRotation=<rot -64.7854 -3.089 6.0959>, targetTranslation=<15.5111 0 0>}]
      customCurve {defaultValue=3.4028234663852886e+38, postInfinityExtrap=RCCE_Constant, preInfinityExtrap=RCCE_Constant, keys []}
      targetRotation=<rot 0 0 0>
     }
     {
      bApplyCustomCurve=false
      bIsHidden=false
      distanceMethod=DefaultMethod
      drivenName=arms_out_skydiving
      functionType=DefaultFunction
      targetScale=0.7112
      boneTransforms [{targetRotation=<rot -26.1816 -8.0177 -68.9757>, targetTranslation=<15.5111 0 0>}]
      customCurve {defaultValue=3.4028234663852886e+38, postInfinityExtrap=RCCE_Constant, preInfinityExtrap=RCCE_Constant, keys []}
      targetRotation=<rot 0 0 0>
     }
     {
      bApplyCustomCurve=false
      bIsHidden=false
      distanceMethod=DefaultMethod
      drivenName=arms_up_skydiving
      functionType=DefaultFunction
      targetScale=2
      boneTransforms [{targetRotation=<rot 17.5036 16.2169 -69.177>, targetTranslation=<15.5111 0 0>}]
      customCurve {defaultValue=3.4028234663852886e+38, postInfinityExtrap=RCCE_Constant, preInfinityExtrap=RCCE_Constant, keys []}
      targetRotation=<rot 0 0 0>
     }
     {
      bApplyCustomCurve=false
      bIsHidden=false
      distanceMethod=DefaultMethod
      drivenName=arms_back_skydiving
      functionType=DefaultFunction
      targetScale=0.6771
      boneTransforms [{targetRotation=<rot -41.801 -46.7216 34.921>, targetTranslation=<15.5111 0 0>}]
      customCurve {defaultValue=3.4028234663852886e+38, postInfinityExtrap=RCCE_Constant, preInfinityExtrap=RCCE_Constant, keys []}
      targetRotation=<rot 0 0 0>
     }
    ]
    rBFParams {radius=24.815}
    sourceBones [{boneName=upperarm_l}]
   }
   n3 = Pose Driver - Source: clavicle_r!(SourcePose:Pose Link Structure=Use cached pose 'CopyPose_Cache'()) -> (Pose:Pose Link Structure) {
    is=title AnimGraphNode_PoseDriver
    poseAsset=&/M_MED_Dark_Shaman_Grounded_RBF_PoseAsset
    onlyDriveBones [
     {boneName=dyn_shoulderpad_top_pivot_r}
     {boneName=dyn_shoulderpad_master_r}
     {boneName=dyn_shoulderpad_bottom_pivot_r}
     {boneName=dyn_shoulderpad_front_pivot_r}
     {boneName=dyn_shoulderpad_back_pivot_r}
     {boneName=dyn_shoulderpad_top_r}
     {boneName=dyn_shoulderpad_back_01_r}
     {boneName=dyn_shoulderpad_back_02_r}
     {boneName=dyn_shoulderpad_mid_01_r}
     {boneName=dyn_shoulderpad_mid_02_r}
     {boneName=dyn_shoulderpad_front_01_r}
     {boneName=dyn_shoulderpad_front_02_r}
    ]
    poseTargets [
     {
      bApplyCustomCurve=false
      bIsHidden=false
      distanceMethod=DefaultMethod
      drivenName=Pose_0
      functionType=DefaultFunction
      targetScale=0.4809
      boneTransforms [
       {targetRotation=<rot -85.1914 41.4375 143.0618>, targetTranslation=<4.3575 -0.1466 2.3223>}
       {targetRotation=<rot -64.7854 -3.089 6.0959>, targetTranslation=<-15.5111 0 0>}
      ]
      customCurve {defaultValue=3.4028234663852886e+38, postInfinityExtrap=RCCE_Constant, preInfinityExtrap=RCCE_Constant, keys []}
      targetRotation=<rot 0 0 0>
     }
     {
      bApplyCustomCurve=false
      bIsHidden=false
      distanceMethod=DefaultMethod
      drivenName=Pose_1
      functionType=DefaultFunction
      targetScale=0.6847
      boneTransforms [
       {targetRotation=<rot -64.2747 -120.6598 -83.241>, targetTranslation=<4.3575 -0.1466 2.3223>}
       {targetRotation=<rot -23.3709 86.7554 -36.3331>, targetTranslation=<-15.5111 0 0>}
      ]
      customCurve {defaultValue=3.4028234663852886e+38, postInfinityExtrap=RCCE_Constant, preInfinityExtrap=RCCE_Constant, keys []}
      targetRotation=<rot 0 0 0>
     }
     {
      bApplyCustomCurve=false
      bIsHidden=false
      distanceMethod=DefaultMethod
      drivenName=Pose_2
      functionType=DefaultFunction
      targetScale=0.3773
      boneTransforms [
       {targetRotation=<rot -69.838 165.5965 22.6769>, targetTranslation=<4.3575 -0.1466 2.3223>}
       {targetRotation=<rot -75.1088 43.0245 -78.2219>, targetTranslation=<-15.5111 0 0>}
      ]
      customCurve {defaultValue=3.4028234663852886e+38, postInfinityExtrap=RCCE_Constant, preInfinityExtrap=RCCE_Constant, keys []}
      targetRotation=<rot 0 0 0>
     }
     {
      bApplyCustomCurve=false
      bIsHidden=false
      distanceMethod=DefaultMethod
      drivenName=Pose_3
      functionType=DefaultFunction
      targetScale=0.3773
      boneTransforms [
       {targetRotation=<rot -73.4134 -176.9703 4.3592>, targetTranslation=<4.3575 -0.1466 2.3223>}
       {targetRotation=<rot -75.5656 105.7208 -93.8077>, targetTranslation=<-15.5111 0 0>}
      ]
      customCurve {defaultValue=3.4028234663852886e+38, postInfinityExtrap=RCCE_Constant, preInfinityExtrap=RCCE_Constant, keys []}
      targetRotation=<rot 0 0 0>
     }
     {
      bApplyCustomCurve=false
      bIsHidden=false
      distanceMethod=DefaultMethod
      drivenName=Pose_4
      functionType=DefaultFunction
      targetScale=0.4809
      boneTransforms [
       {targetRotation=<rot -87.7993 76.2682 109.0758>, targetTranslation=<4.3575 -0.1466 2.3223>}
       {targetRotation=<rot -69.7589 51.1037 -35.2693>, targetTranslation=<-15.5111 0 0>}
      ]
      customCurve {defaultValue=3.4028234663852886e+38, postInfinityExtrap=RCCE_Constant, preInfinityExtrap=RCCE_Constant, keys []}
      targetRotation=<rot 0 0 0>
     }
     {
      bApplyCustomCurve=false
      bIsHidden=false
      distanceMethod=DefaultMethod
      drivenName=Pose_5
      functionType=DefaultFunction
      targetScale=0.3608
      boneTransforms [
       {targetRotation=<rot -76.2754 172.1506 9.5506>, targetTranslation=<4.3575 -0.1466 2.3223>}
       {targetRotation=<rot -45.0185 104.3596 -56.1379>, targetTranslation=<-15.5111 0 0>}
      ]
      customCurve {defaultValue=3.4028234663852886e+38, postInfinityExtrap=RCCE_Constant, preInfinityExtrap=RCCE_Constant, keys []}
      targetRotation=<rot 0 0 0>
     }
     {
      bApplyCustomCurve=false
      bIsHidden=false
      distanceMethod=DefaultMethod
      drivenName=Pose_6
      functionType=DefaultFunction
      targetScale=0.3608
      boneTransforms [
       {targetRotation=<rot -74.8686 -132.7057 -46.9771>, targetTranslation=<4.3575 -0.1466 2.3223>}
       {targetRotation=<rot -39.6825 113.2505 -70.71>, targetTranslation=<-15.5111 0 0>}
      ]
      customCurve {defaultValue=3.4028234663852886e+38, postInfinityExtrap=RCCE_Constant, preInfinityExtrap=RCCE_Constant, keys []}
      targetRotation=<rot 0 0 0>
     }
     {
      bApplyCustomCurve=false
      bIsHidden=false
      distanceMethod=DefaultMethod
      drivenName=Pose_7
      functionType=DefaultFunction
      targetScale=0.4435
      boneTransforms [
       {targetRotation=<rot -57.8419 -161.5812 -18.7948>, targetTranslation=<4.3575 -0.1466 2.3223>}
       {targetRotation=<rot -66.8684 120.4314 -71.1722>, targetTranslation=<-15.5111 0 0>}
      ]
      customCurve {
       defaultValue=3.4028234663852886e+38
       postInfinityExtrap=RCCE_Constant
       preInfinityExtrap=RCCE_Constant
       keys [
        {
         arriveTangent=0
         arriveTangentWeight=0
         interpMode=RCIM_Linear
         leaveTangent=0
         leaveTangentWeight=0
         tangentMode=RCTM_Auto
         tangentWeightMode=RCTWM_WeightedNone
         time=0
         value=0
        }
        {
         arriveTangent=0
         arriveTangentWeight=0
         interpMode=RCIM_Linear
         leaveTangent=0
         leaveTangentWeight=0
         tangentMode=RCTM_Auto
         tangentWeightMode=RCTWM_WeightedNone
         time=1
         value=1
        }
       ]
      }
      targetRotation=<rot 0 0 0>
     }
     {
      bApplyCustomCurve=false
      bIsHidden=false
      distanceMethod=DefaultMethod
      drivenName=Pose_8
      functionType=DefaultFunction
      targetScale=0.9584
      boneTransforms [
       {targetRotation=<rot -53.6465 -152.7178 -29.1383>, targetTranslation=<4.3575 -0.1466 2.3223>}
       {targetRotation=<rot -17.8084 26.8504 -97.253>, targetTranslation=<-15.5111 0 0>}
      ]
      customCurve {defaultValue=3.4028234663852886e+38, postInfinityExtrap=RCCE_Constant, preInfinityExtrap=RCCE_Constant, keys []}
      targetRotation=<rot 0 0 0>
     }
     {
      bApplyCustomCurve=false
      bIsHidden=false
      distanceMethod=DefaultMethod
      drivenName=Pose_9
      functionType=DefaultFunction
      targetScale=2
      boneTransforms [
       {targetRotation=<rot -68.9933 -127.6298 -53.9107>, targetTranslation=<4.3575 -0.1466 2.3223>}
       {targetRotation=<rot 33.2732 89.3676 -89.1674>, targetTranslation=<-15.5111 0 0>}
      ]
      customCurve {
       defaultValue=3.4028234663852886e+38
       postInfinityExtrap=RCCE_Constant
       preInfinityExtrap=RCCE_Constant
       keys [
        {
         arriveTangent=0
         arriveTangentWeight=0
         interpMode=RCIM_Linear
         leaveTangent=0
         leaveTangentWeight=0
         tangentMode=RCTM_Auto
         tangentWeightMode=RCTWM_WeightedNone
         time=0
         value=0
        }
        {
         arriveTangent=0
         arriveTangentWeight=0
         interpMode=RCIM_Linear
         leaveTangent=0
         leaveTangentWeight=0
         tangentMode=RCTM_Auto
         tangentWeightMode=RCTWM_WeightedNone
         time=0.3415
         value=0.8068
        }
        {
         arriveTangent=0
         arriveTangentWeight=0
         interpMode=RCIM_Linear
         leaveTangent=0
         leaveTangentWeight=0
         tangentMode=RCTM_Auto
         tangentWeightMode=RCTWM_WeightedNone
         time=1
         value=1
        }
       ]
      }
      targetRotation=<rot 0 0 0>
     }
     {
      bApplyCustomCurve=false
      bIsHidden=false
      distanceMethod=DefaultMethod
      drivenName=Pose_10
      functionType=DefaultFunction
      targetScale=0.9584
      boneTransforms [
       {targetRotation=<rot -83.622 69.4344 115.2173>, targetTranslation=<4.3575 -0.1466 2.3223>}
       {targetRotation=<rot -11.4935 30.1498 -76.7725>, targetTranslation=<-15.5111 0 0>}
      ]
      customCurve {defaultValue=3.4028234663852886e+38, postInfinityExtrap=RCCE_Constant, preInfinityExtrap=RCCE_Constant, keys []}
      targetRotation=<rot 0 0 0>
     }
     {
      bApplyCustomCurve=false
      bIsHidden=false
      distanceMethod=DefaultMethod
      drivenName=Pose_11
      functionType=DefaultFunction
      targetScale=1
      boneTransforms [
       {targetRotation=<rot -66.8403 -177.8105 2.9442>, targetTranslation=<4.3575 -0.1466 2.3223>}
       {targetRotation=<rot 18.2092 9.4898 -75.6901>, targetTranslation=<-15.5111 0 0>}
      ]
      customCurve {defaultValue=3.4028234663852886e+38, postInfinityExtrap=RCCE_Constant, preInfinityExtrap=RCCE_Constant, keys []}
      targetRotation=<rot 0 0 0>
     }
     {
      bApplyCustomCurve=false
      bIsHidden=false
      distanceMethod=DefaultMethod
      drivenName=Pose_19
      functionType=DefaultFunction
      targetScale=0.724
      boneTransforms [
       {targetRotation=<rot -47.9536 -145.5199 -31.7961>, targetTranslation=<4.3575 -0.1466 2.3223>}
       {targetRotation=<rot -20.6591 112.0123 -37.7437>, targetTranslation=<-15.5111 0 0>}
      ]
      customCurve {defaultValue=3.4028234663852886e+38, postInfinityExtrap=RCCE_Constant, preInfinityExtrap=RCCE_Constant, keys []}
      targetRotation=<rot 0 0 0>
     }
     {
      bApplyCustomCurve=false
      bIsHidden=false
      distanceMethod=DefaultMethod
      drivenName=Pose_20
      functionType=DefaultFunction
      targetScale=0.7854
      boneTransforms [
       {targetRotation=<rot -86.6669 -153.8735 -21.7279>, targetTranslation=<4.3575 -0.1466 2.3223>}
       {targetRotation=<rot 61.9226 97.9326 -80.7543>, targetTranslation=<-15.5111 0 0>}
      ]
      customCurve {defaultValue=3.4028234663852886e+38, postInfinityExtrap=RCCE_Constant, preInfinityExtrap=RCCE_Constant, keys []}
      targetRotation=<rot 0 0 0>
     }
    ]
    rBFParams {radius=24}
    sourceBones [{boneName=clavicle_r}, {boneName=upperarm_r}]
   }
   n4 = Pose Driver - Source: clavicle_l!(SourcePose:Pose Link Structure=@n3) -> (Pose:Pose Link Structure) {
    is=title AnimGraphNode_PoseDriver
    poseAsset=&M_MED_Dark_Shaman_Grounded_RBF_PoseAsset
    onlyDriveBones [
     {boneName=dyn_shoulderpad_top_pivot_l}
     {boneName=dyn_shoulderpad_master_l}
     {boneName=dyn_shoulderpad_bottom_pivot_l}
     {boneName=dyn_shoulderpad_front_pivot_l}
     {boneName=dyn_shoulderpad_back_pivot_l}
     {boneName=dyn_shoulderpad_top_l}
     {boneName=dyn_shoulderpad_back_01_l}
     {boneName=dyn_shoulderpad_back_02_l}
     {boneName=dyn_shoulderpad_mid_01_l}
     {boneName=dyn_shoulderpad_mid_02_l}
     {boneName=dyn_shoulderpad_front_01_l}
     {boneName=dyn_shoulderpad_front_02_l}
    ]
    poseTargets [
     {
      bApplyCustomCurve=false
      bIsHidden=false
      distanceMethod=DefaultMethod
      drivenName=Pose_0
      functionType=DefaultFunction
      targetScale=0.3972
      boneTransforms [
       {targetRotation=<rot -85.1913 -138.5624 143.0618>, targetTranslation=<4.3575 -0.1466 -2.3223>}
       {targetRotation=<rot -64.7854 -3.089 6.0959>, targetTranslation=<15.5111 0 0>}
      ]
      customCurve {defaultValue=3.4028234663852886e+38, postInfinityExtrap=RCCE_Constant, preInfinityExtrap=RCCE_Constant, keys []}
      targetRotation=<rot 0 0 0>
     }
     {
      bApplyCustomCurve=false
      bIsHidden=false
      distanceMethod=DefaultMethod
      drivenName=Pose_1
      functionType=DefaultFunction
      targetScale=0.871
      boneTransforms [
       {targetRotation=<rot -71.8075 -7.0774 -33.3057>, targetTranslation=<4.3575 -0.1466 -2.3223>}
       {targetRotation=<rot -9.5233 -21.9053 -28.615>, targetTranslation=<15.5111 0 0>}
      ]
      customCurve {defaultValue=3.4028234663852886e+38, postInfinityExtrap=RCCE_Constant, preInfinityExtrap=RCCE_Constant, keys []}
      targetRotation=<rot 0 0 0>
     }
     {
      bApplyCustomCurve=false
      bIsHidden=false
      distanceMethod=DefaultMethod
      drivenName=Pose_2
      functionType=DefaultFunction
      targetScale=0.5255
      boneTransforms [
       {targetRotation=<rot -75.4618 19.5667 -12.8773>, targetTranslation=<4.3575 -0.1466 -2.3223>}
       {targetRotation=<rot -64.9089 79.0369 -47.2079>, targetTranslation=<15.5111 0 0>}
      ]
      customCurve {defaultValue=3.4028234663852886e+38, postInfinityExtrap=RCCE_Constant, preInfinityExtrap=RCCE_Constant, keys []}
      targetRotation=<rot 0 0 0>
     }
     {
      bApplyCustomCurve=false
      bIsHidden=false
      distanceMethod=DefaultMethod
      drivenName=Pose_3
      functionType=DefaultFunction
      targetScale=0.6912
      boneTransforms [
       {targetRotation=<rot -74.7073 55.0349 -49.5566>, targetTranslation=<4.3575 -0.1466 -2.3223>}
       {targetRotation=<rot -36.5188 -9.2495 27.844>, targetTranslation=<15.5111 0 0>}
      ]
      customCurve {defaultValue=3.4028234663852886e+38, postInfinityExtrap=RCCE_Constant, preInfinityExtrap=RCCE_Constant, keys []}
      targetRotation=<rot 0 0 0>
     }
     {
      bApplyCustomCurve=false
      bIsHidden=false
      distanceMethod=DefaultMethod
      drivenName=Pose_4
      functionType=DefaultFunction
      targetScale=0.3972
      boneTransforms [
       {targetRotation=<rot -81.7412 -135.4937 140.2621>, targetTranslation=<4.3575 -0.1466 -2.3223>}
       {targetRotation=<rot -49.8941 -8.1717 1.1726>, targetTranslation=<15.5111 0 0>}
      ]
      customCurve {defaultValue=3.4028234663852886e+38, postInfinityExtrap=RCCE_Constant, preInfinityExtrap=RCCE_Constant, keys []}
      targetRotation=<rot 0 0 0>
     }
     {
      bApplyCustomCurve=false
      bIsHidden=false
      distanceMethod=DefaultMethod
      drivenName=Pose_5
      functionType=DefaultFunction
      targetScale=0.8882
      boneTransforms [
       {targetRotation=<rot -80.8623 39.04 -38.1229>, targetTranslation=<4.3575 -0.1466 -2.3223>}
       {targetRotation=<rot -31.8256 65.8251 -86.0604>, targetTranslation=<15.5111 0 0>}
      ]
      customCurve {defaultValue=3.4028234663852886e+38, postInfinityExtrap=RCCE_Constant, preInfinityExtrap=RCCE_Constant, keys []}
      targetRotation=<rot 0 0 0>
     }
     {
      bApplyCustomCurve=false
      bIsHidden=false
      distanceMethod=DefaultMethod
      drivenName=Pose_6
      functionType=DefaultFunction
      targetScale=0.9332
      boneTransforms [
       {targetRotation=<rot -47.5697 14.692 -0.434>, targetTranslation=<4.3575 -0.1466 -2.3223>}
       {targetRotation=<rot -8.3098 26.4205 -78.9654>, targetTranslation=<15.5111 0 0>}
      ]
      customCurve {defaultValue=3.4028234663852886e+38, postInfinityExtrap=RCCE_Constant, preInfinityExtrap=RCCE_Constant, keys []}
      targetRotation=<rot 0 0 0>
     }
     {
      bApplyCustomCurve=false
      bIsHidden=false
      distanceMethod=DefaultMethod
      drivenName=Pose_7
      functionType=DefaultFunction
      targetScale=0.5255
      boneTransforms [
       {targetRotation=<rot -86.0332 143.397 -137.7672>, targetTranslation=<4.3575 -0.1466 -2.3223>}
       {targetRotation=<rot -67.1606 51.1714 -32.7303>, targetTranslation=<15.5111 0 0>}
      ]
      customCurve {defaultValue=3.4028234663852886e+38, postInfinityExtrap=RCCE_Constant, preInfinityExtrap=RCCE_Constant, keys []}
      targetRotation=<rot 0 0 0>
     }
     {
      bApplyCustomCurve=false
      bIsHidden=false
      distanceMethod=DefaultMethod
      drivenName=Pose_8
      functionType=DefaultFunction
      targetScale=0.4637
      boneTransforms [
       {targetRotation=<rot -84.0122 51.6419 -47.6755>, targetTranslation=<4.3575 -0.1466 -2.3223>}
       {targetRotation=<rot -74.2196 -36.1046 27.6902>, targetTranslation=<15.5111 0 0>}
      ]
      customCurve {defaultValue=3.4028234663852886e+38, postInfinityExtrap=RCCE_Constant, preInfinityExtrap=RCCE_Constant, keys []}
      targetRotation=<rot 0 0 0>
     }
     {
      bApplyCustomCurve=false
      bIsHidden=false
      distanceMethod=DefaultMethod
      drivenName=Pose_9
      functionType=DefaultFunction
      targetScale=1.5
      boneTransforms [
       {targetRotation=<rot -48.3561 10.1188 -2.4428>, targetTranslation=<4.3575 -0.1466 -2.3223>}
       {targetRotation=<rot 14.5763 54.8781 -98.5188>, targetTranslation=<15.5111 0 0>}
      ]
      customCurve {defaultValue=3.4028234663852886e+38, postInfinityExtrap=RCCE_Constant, preInfinityExtrap=RCCE_Constant, keys []}
      targetRotation=<rot 0 0 0>
     }
     {
      bApplyCustomCurve=false
      bIsHidden=false
      distanceMethod=DefaultMethod
      drivenName=Pose_10
      functionType=DefaultFunction
      targetScale=0.9994
      boneTransforms [
       {targetRotation=<rot -79.2873 99.6394 -110.9318>, targetTranslation=<4.3575 -0.1466 -2.3223>}
       {targetRotation=<rot -14.9047 102.8909 -17.6919>, targetTranslation=<15.5111 0 0>}
      ]
      customCurve {defaultValue=3.4028234663852886e+38, postInfinityExtrap=RCCE_Constant, preInfinityExtrap=RCCE_Constant, keys []}
      targetRotation=<rot 0 0 0>
     }
     {
      bApplyCustomCurve=false
      bIsHidden=false
      distanceMethod=DefaultMethod
      drivenName=Pose_11
      functionType=DefaultFunction
      targetScale=0.9076
      boneTransforms [
       {targetRotation=<rot -61.5845 27.1483 -22.337>, targetTranslation=<4.3575 -0.1466 -2.3223>}
       {targetRotation=<rot 18.262 -7.1845 -72.1573>, targetTranslation=<15.5111 0 0>}
      ]
      customCurve {defaultValue=3.4028234663852886e+38, postInfinityExtrap=RCCE_Constant, preInfinityExtrap=RCCE_Constant, keys []}
      targetRotation=<rot 0 0 0>
     }
     {
      bApplyCustomCurve=false
      bIsHidden=false
      distanceMethod=DefaultMethod
      drivenName=Pose_19
      functionType=DefaultFunction
      targetScale=1
      boneTransforms [
       {targetRotation=<rot -58.4846 -9.5373 2.6759>, targetTranslation=<4.3575 -0.1466 -2.3223>}
       {targetRotation=<rot -44.7022 128.4906 -48.0679>, targetTranslation=<15.5111 0 0>}
      ]
      customCurve {defaultValue=3.4028234663852886e+38, postInfinityExtrap=RCCE_Constant, preInfinityExtrap=RCCE_Constant, keys []}
      targetRotation=<rot 0 0 0>
     }
    ]
    rBFParams {radius=25}
    sourceBones [{boneName=clavicle_l}, {boneName=upperarm_l}]
   }
   n5 = Get bIsSkyDiving()
   n6 = Anim Dynamics - Bone: dyn_llama(
    ComponentPose=Local To Component(LocalPose=Blend Poses by bool(BlendPose_0=@n2, BlendPose_1=@n4, bActiveValue=@n5) {blendType=Linear})
    Alpha=1
    ExternalForce="0,0,0"
   ) {
    angularDampingOverride=1
    bGravityOverrideInSimSpace=true
    bOverrideAngularDamping=true
    simulationSpace=RootRelative
    alphaBoolBlend {blendOption=Linear}
    boundBone {boneName=dyn_llama}
    physicsBodyDefinitions [{
      sphereCollisionRadius=0
      boundBone {boneName=dyn_llama}
      boxExtents=<20 20 5>
      constraintSetup {angularTargetAxis=AxisZ, twistAxis=AxisZ, angularLimitsMax=<30 0 0>, angularLimitsMin=<-30 0 0>, angularTarget=<0 0 0>}
      localJointOffset=<0 0 -5>
     }]
    relativeSpaceBone {boneName=root}
   }
   n7 = "Transform (Modify) Bone - Bone: dyn_llama"(
    ComponentPose=@n6
    Alpha=Get thigh_l_crouch() -> (thigh_l_crouch:double)
    Translation="-0.177,-4.996,0.000"
    Rotation="-0.868,-0.193,24.985"
    Scale="1,1,1"
   ) {
    rotationMode=BMM_Additive
    rotationSpace=BCS_ParentBoneSpace
    translationMode=BMM_Additive
    translationSpace=BCS_ParentBoneSpace
    alphaBoolBlend {blendOption=Linear}
    boneToModify {boneName=dyn_llama}
    rotation=<rot -0.868 -0.193 24.985>
    translation=<-0.177 -4.996>
   }
   n8 = Break Fort Anim Input Skydiving External Force(FortAnimInput_SkydivingExternalForce=Get SkydivingExternalForceInput()) -> (
    bUseSkydivingVectorForce=~
    SkydivingVectorForce=~
    HeadToPelvisDirection=~
    FloatingMultiplier=~
    FloatingAdditive=~
    DivingMultiplier=~
    DivingAdditive=~
    ParachutingMultiplier=~
    ParachutingAdditive=~
   )
   n9 = RigidBody(ComponentPose=@n7, ExternalForce=@n8.SkydivingVectorForce) {
    bClampLinearTranslationLimitToRefPose=true
    overridePhysicsAsset=&/M_MED_Dark_Shaman_Physics
    simulationSpace=BaseBoneSpace
    alphaBoolBlend {blendOption=Linear}
    baseBoneRef {boneName=root}
    componentLinearAccScale=<1 1 1>
    componentLinearVelScale=<1 1 1>
   }
   n10 = Anim Dynamics - Bone: dyn_leg_skirt_02_r(ComponentPose=@n9, Alpha=1, ExternalForce=@n8.SkydivingVectorForce) {
    angularDampingOverride=1
    bGravityOverrideInSimSpace=true
    bOverrideAngularDamping=true
    simulationSpace=RootRelative
    alphaBoolBlend {blendOption=Linear}
    boundBone {boneName=dyn_leg_skirt_02_r}
    physicsBodyDefinitions [{
      sphereCollisionRadius=0
      boundBone {boneName=dyn_leg_skirt_02_r}
      boxExtents=<5 20 20>
      constraintSetup {angularLimitsMax=<0 30 0>, angularLimitsMin=<0 0 0>, angularTarget=<0 0 0>}
      localJointOffset=<5 0 0>
     }]
    relativeSpaceBone {boneName=root}
   }
   n11 = Anim Dynamics - Bone: dyn_leg_skirt_03_r(ComponentPose=@n10, Alpha=1, ExternalForce=@n8.SkydivingVectorForce) {
    angularDampingOverride=1
    bGravityOverrideInSimSpace=true
    bOverrideAngularDamping=true
    simulationSpace=RootRelative
    alphaBoolBlend {blendOption=Linear}
    boundBone {boneName=dyn_leg_skirt_03_r}
    physicsBodyDefinitions [{
      sphereCollisionRadius=0
      boundBone {boneName=dyn_leg_skirt_03_r}
      boxExtents=<5 20 20>
      constraintSetup {angularLimitsMax=<0 30 5>, angularLimitsMin=<0 0 -5>, angularTarget=<0 0 0>}
      localJointOffset=<5 0 0>
     }]
    relativeSpaceBone {boneName=root}
   }
   n12 = Anim Dynamics - Bone: dyn_leg_skirt_04_r(ComponentPose=@n11, Alpha=1, ExternalForce=@n8.SkydivingVectorForce) {
    angularDampingOverride=1
    bGravityOverrideInSimSpace=true
    bOverrideAngularDamping=true
    simulationSpace=RootRelative
    alphaBoolBlend {blendOption=Linear}
    boundBone {boneName=dyn_leg_skirt_04_r}
    physicsBodyDefinitions [{
      sphereCollisionRadius=0
      boundBone {boneName=dyn_leg_skirt_04_r}
      boxExtents=<5 20 20>
      constraintSetup {angularLimitsMax=<0 30 5>, angularLimitsMin=<0 -5 -5>, angularTarget=<0 0 0>}
      localJointOffset=<5 0 0>
     }]
    relativeSpaceBone {boneName=root}
   }
   n13 = Anim Dynamics - Bone: dyn_leg_skirt_02_l(ComponentPose=@n12, Alpha=1, ExternalForce=@n8.SkydivingVectorForce) {
    angularDampingOverride=1
    bGravityOverrideInSimSpace=true
    bOverrideAngularDamping=true
    simulationSpace=RootRelative
    alphaBoolBlend {blendOption=Linear}
    boundBone {boneName=dyn_leg_skirt_02_l}
    physicsBodyDefinitions [{
      sphereCollisionRadius=0
      boundBone {boneName=dyn_leg_skirt_02_l}
      boxExtents=<5 20 20>
      constraintSetup {angularLimitsMax=<0 30 0>, angularLimitsMin=<0 0 0>, angularTarget=<0 0 0>}
      localJointOffset=<5 0 0>
     }]
    relativeSpaceBone {boneName=root}
   }
   n14 = Anim Dynamics - Bone: dyn_leg_skirt_03_l(ComponentPose=@n13, Alpha=1, ExternalForce=@n8.SkydivingVectorForce) {
    angularDampingOverride=1
    bGravityOverrideInSimSpace=true
    bOverrideAngularDamping=true
    simulationSpace=RootRelative
    alphaBoolBlend {blendOption=Linear}
    boundBone {boneName=dyn_leg_skirt_03_l}
    physicsBodyDefinitions [{
      sphereCollisionRadius=0
      boundBone {boneName=dyn_leg_skirt_03_l}
      boxExtents=<5 20 20>
      constraintSetup {angularLimitsMax=<0 30 5>, angularLimitsMin=<0 0 -5>, angularTarget=<0 0 0>}
      localJointOffset=<5 0 0>
     }]
    relativeSpaceBone {boneName=root}
   }
   n15 = Anim Dynamics - Bone: dyn_leg_skirt_04_l(ComponentPose=@n14, Alpha=1, ExternalForce=@n8.SkydivingVectorForce) {
    angularDampingOverride=1
    bGravityOverrideInSimSpace=true
    bOverrideAngularDamping=true
    simulationSpace=RootRelative
    alphaBoolBlend {blendOption=Linear}
    boundBone {boneName=dyn_leg_skirt_04_l}
    physicsBodyDefinitions [{
      sphereCollisionRadius=0
      boundBone {boneName=dyn_leg_skirt_04_l}
      boxExtents=<5 20 20>
      constraintSetup {angularLimitsMax=<0 30 5>, angularLimitsMin=<0 -5 -5>, angularTarget=<0 0 0>}
      localJointOffset=<5 0 0>
     }]
    relativeSpaceBone {boneName=root}
   }
   n16 = Anim Dynamics - Bone: dyn_shoulder_fur_01_l(ComponentPose=@n15, Alpha=1, ExternalForce=@n8.SkydivingVectorForce) {
    angularDampingOverride=1
    bGravityOverrideInSimSpace=true
    bOverrideAngularDamping=true
    simulationSpace=RootRelative
    alphaBoolBlend {blendOption=Linear}
    boundBone {boneName=dyn_shoulder_fur_01_l}
    componentLinearAccScale=<1 1 2>
    componentLinearVelScale=<1 1 2>
    physicsBodyDefinitions [{
      sphereCollisionRadius=0
      boundBone {boneName=dyn_shoulder_fur_01_l}
      boxExtents=<5 20 20>
      constraintSetup {angularLimitsMax=<0 30 5>, angularLimitsMin=<0 0 -5>, angularTarget=<0 0 0>}
      localJointOffset=<5 0 0>
     }]
    relativeSpaceBone {boneName=root}
   }
   n17 = Anim Dynamics - Bone: dyn_shoulder_fur_02_l(ComponentPose=@n16, Alpha=1, ExternalForce=@n8.SkydivingVectorForce) {
    angularDampingOverride=1
    bGravityOverrideInSimSpace=true
    bOverrideAngularDamping=true
    simulationSpace=RootRelative
    alphaBoolBlend {blendOption=Linear}
    boundBone {boneName=dyn_shoulder_fur_02_l}
    componentLinearAccScale=<1 1 2>
    componentLinearVelScale=<1 1 2>
    physicsBodyDefinitions [{
      sphereCollisionRadius=0
      boundBone {boneName=dyn_shoulder_fur_02_l}
      boxExtents=<5 20 20>
      constraintSetup {angularLimitsMax=<0 30 5>, angularLimitsMin=<0 0 -5>, angularTarget=<0 0 0>}
      localJointOffset=<5 0 0>
     }]
    relativeSpaceBone {boneName=root}
   }
   n18 = Anim Dynamics - Bone: dyn_shoulder_fur_03_l(ComponentPose=@n17, Alpha=1, ExternalForce=@n8.SkydivingVectorForce) {
    angularDampingOverride=1
    bGravityOverrideInSimSpace=true
    bOverrideAngularDamping=true
    simulationSpace=RootRelative
    alphaBoolBlend {blendOption=Linear}
    boundBone {boneName=dyn_shoulder_fur_03_l}
    componentLinearAccScale=<1 1 2>
    componentLinearVelScale=<1 1 2>
    physicsBodyDefinitions [{
      sphereCollisionRadius=0
      boundBone {boneName=dyn_shoulder_fur_03_l}
      boxExtents=<5 20 20>
      constraintSetup {angularLimitsMax=<0 30 5>, angularLimitsMin=<0 0 -5>, angularTarget=<0 0 0>}
      localJointOffset=<5 0 0>
     }]
    relativeSpaceBone {boneName=root}
   }
   n19 = Anim Dynamics - Bone: dyn_shoulder_fur_04_l(ComponentPose=@n18, Alpha=1, ExternalForce=@n8.SkydivingVectorForce) {
    angularDampingOverride=1
    bGravityOverrideInSimSpace=true
    bOverrideAngularDamping=true
    simulationSpace=RootRelative
    alphaBoolBlend {blendOption=Linear}
    boundBone {boneName=dyn_shoulder_fur_04_l}
    componentLinearAccScale=<1 1 2>
    componentLinearVelScale=<1 1 2>
    physicsBodyDefinitions [{
      sphereCollisionRadius=0
      boundBone {boneName=dyn_shoulder_fur_04_l}
      boxExtents=<5 20 20>
      constraintSetup {angularLimitsMax=<0 30 5>, angularLimitsMin=<0 0 -5>, angularTarget=<0 0 0>}
      localJointOffset=<5 0 0>
     }]
    relativeSpaceBone {boneName=root}
   }
   n20 = Anim Dynamics - Bone: dyn_shoulder_fur_05_l(ComponentPose=@n19, Alpha=1, ExternalForce=@n8.SkydivingVectorForce) {
    angularDampingOverride=1
    bGravityOverrideInSimSpace=true
    bOverrideAngularDamping=true
    simulationSpace=RootRelative
    alphaBoolBlend {blendOption=Linear}
    boundBone {boneName=dyn_shoulder_fur_05_l}
    componentLinearAccScale=<1 1 2>
    componentLinearVelScale=<1 1 2>
    physicsBodyDefinitions [{
      sphereCollisionRadius=0
      boundBone {boneName=dyn_shoulder_fur_05_l}
      boxExtents=<5 20 20>
      constraintSetup {angularLimitsMax=<0 30 5>, angularLimitsMin=<0 0 -5>, angularTarget=<0 0 0>}
      localJointOffset=<5 0 0>
     }]
    relativeSpaceBone {boneName=root}
   }
   n21 = Anim Dynamics - Bone: dyn_shoulder_fur_01_r(ComponentPose=@n20, Alpha=1, ExternalForce=@n8.SkydivingVectorForce) {
    angularDampingOverride=1
    bGravityOverrideInSimSpace=true
    bOverrideAngularDamping=true
    simulationSpace=RootRelative
    alphaBoolBlend {blendOption=Linear}
    boundBone {boneName=dyn_shoulder_fur_01_r}
    componentLinearAccScale=<1 1 2>
    componentLinearVelScale=<1 1 2>
    physicsBodyDefinitions [{
      sphereCollisionRadius=0
      boundBone {boneName=dyn_shoulder_fur_01_r}
      boxExtents=<5 20 20>
      constraintSetup {angularLimitsMax=<0 30 5>, angularLimitsMin=<0 0 -5>, angularTarget=<0 0 0>}
      localJointOffset=<5 0 0>
     }]
    relativeSpaceBone {boneName=root}
   }
   n22 = Anim Dynamics - Bone: dyn_shoulder_fur_02_r(ComponentPose=@n21, Alpha=1, ExternalForce=@n8.SkydivingVectorForce) {
    angularDampingOverride=1
    bGravityOverrideInSimSpace=true
    bOverrideAngularDamping=true
    simulationSpace=RootRelative
    alphaBoolBlend {blendOption=Linear}
    boundBone {boneName=dyn_shoulder_fur_02_r}
    componentLinearAccScale=<1 1 2>
    componentLinearVelScale=<1 1 2>
    physicsBodyDefinitions [{
      sphereCollisionRadius=0
      boundBone {boneName=dyn_shoulder_fur_02_r}
      boxExtents=<5 20 20>
      constraintSetup {angularLimitsMax=<0 30 5>, angularLimitsMin=<0 0 -5>, angularTarget=<0 0 0>}
      localJointOffset=<5 0 0>
     }]
    relativeSpaceBone {boneName=root}
   }
   n23 = Anim Dynamics - Bone: dyn_shoulder_fur_03_r(ComponentPose=@n22, Alpha=1, ExternalForce=@n8.SkydivingVectorForce) {
    angularDampingOverride=1
    bGravityOverrideInSimSpace=true
    bOverrideAngularDamping=true
    simulationSpace=RootRelative
    alphaBoolBlend {blendOption=Linear}
    boundBone {boneName=dyn_shoulder_fur_03_r}
    componentLinearAccScale=<1 1 2>
    componentLinearVelScale=<1 1 2>
    physicsBodyDefinitions [{
      sphereCollisionRadius=0
      boundBone {boneName=dyn_shoulder_fur_03_r}
      boxExtents=<5 20 20>
      constraintSetup {angularLimitsMax=<0 30 5>, angularLimitsMin=<0 0 -5>, angularTarget=<0 0 0>}
      localJointOffset=<5 0 0>
     }]
    relativeSpaceBone {boneName=root}
   }
   n24 = Anim Dynamics - Bone: dyn_shoulder_fur_04_r(ComponentPose=@n23, Alpha=1, ExternalForce=@n8.SkydivingVectorForce) {
    angularDampingOverride=1
    bGravityOverrideInSimSpace=true
    bOverrideAngularDamping=true
    simulationSpace=RootRelative
    alphaBoolBlend {blendOption=Linear}
    boundBone {boneName=dyn_shoulder_fur_04_r}
    componentLinearAccScale=<1 1 2>
    componentLinearVelScale=<1 1 2>
    physicsBodyDefinitions [{
      sphereCollisionRadius=0
      boundBone {boneName=dyn_shoulder_fur_04_r}
      boxExtents=<5 20 20>
      constraintSetup {angularLimitsMax=<0 30 5>, angularLimitsMin=<0 0 -5>, angularTarget=<0 0 0>}
      localJointOffset=<5 0 0>
     }]
    relativeSpaceBone {boneName=root}
   }
   n25 = Anim Dynamics - Bone: dyn_shoulder_fur_05_r(ComponentPose=@n24, Alpha=1, ExternalForce=@n8.SkydivingVectorForce) {
    angularDampingOverride=1
    bGravityOverrideInSimSpace=true
    bOverrideAngularDamping=true
    simulationSpace=RootRelative
    alphaBoolBlend {blendOption=Linear}
    boundBone {boneName=dyn_shoulder_fur_05_r}
    componentLinearAccScale=<1 1 2>
    componentLinearVelScale=<1 1 2>
    physicsBodyDefinitions [{
      sphereCollisionRadius=0
      boundBone {boneName=dyn_shoulder_fur_05_r}
      boxExtents=<5 20 20>
      constraintSetup {angularLimitsMax=<0 30 5>, angularLimitsMin=<0 0 -5>, angularTarget=<0 0 0>}
      localJointOffset=<5 0 0>
     }]
    relativeSpaceBone {boneName=root}
   }
   Output Pose(Result=Component To Local(ComponentPose=@n25)) {layerGroup=None}
   Save cached pose 'CopyPose_Cache'(Pose=Copy Pose From Mesh() {bUseAttachedParent=true}) {cacheName=CopyPose_Cache}
  ]}
 EventGraph {nodes [
   Event BlueprintUpdateAnimation()
   n1 = BreakTransform(InTransform=GetDeltaTransformFromRefPose(self=GetPartSkeletalMeshComponent(), BoneName=thigh_l, BaseName=None))
   > Set thigh_l_crouch(thigh_l_crouch:double=MapRangeClamped(Value=BreakRotator(InRot=@n1.Rotation).Yaw, InRangeA=-20, InRangeB=-100)) -> (Output_Get:double)
  ]}
]
variables [thigh_l_crouch {type=number}]
