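# Physics asset for the Dark Shaman male medium body's face accessory mesh, covering hair and bang bones for secondary motion.
# Small capsules form mirrored left and right strand chains, each rooted to one large head capsule.
# Constraints allow tight swing and almost no twist, with looser swing on the single-bone E strands and back hair.
# Back hair uses fatter capsules and forks, with bones 02 and 03 both constrained to 01.
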PhysicsAsset /BRCosmetics/Characters/Player/Male/Medium/Bodies/M_MED_DarkShaman/Meshes/Parts/M_MED_Dark_Shaman_FaceAcc_Physics
bodies {
 dyn_bangs_A_01_l [{shapeType=Capsule, length=3, radius=1, center=<1.5 0 0>, rotation=<rot 90 0 0>}]
 dyn_bangs_A_01_r [{shapeType=Capsule, length=3, radius=1, center=<1.5 0 0>, rotation=<rot 90 0 0>}]
 dyn_bangs_A_02_l [{shapeType=Capsule, length=3, radius=1, center=<1.5 0 0>, rotation=<rot 90 0 0>}]
 dyn_bangs_A_02_r [{shapeType=Capsule, length=3.75, radius=1, center=<2.5 0 -0.0001>, rotation=<rot 90 0 0>}]
 dyn_bangs_A_03_l [{shapeType=Capsule, length=3, radius=1, center=<1 0 0>, rotation=<rot 90 0 0>}]
 dyn_bangs_A_03_r [{shapeType=Capsule, length=3, radius=1, center=<1.5 0 0>, rotation=<rot 90 0 0>}]
 dyn_bangs_B_01_l [{shapeType=Capsule, length=3, radius=1, center=<1.5 0 0>, rotation=<rot 90 0 0>}]
 dyn_bangs_B_01_r [{shapeType=Capsule, length=3, radius=1, center=<1 0 0>, rotation=<rot 90 0 0>}]
 dyn_bangs_B_02_l [{shapeType=Capsule, length=3, radius=1, center=<1.5 0 0>, rotation=<rot 90 0 0>}]
 dyn_bangs_B_02_r [{shapeType=Capsule, length=3, radius=1, center=<1 0 0>, rotation=<rot 90 0 0>}]
 dyn_bangs_B_03_l [{shapeType=Capsule, length=3, radius=1, center=<1.5 0 0>, rotation=<rot 90 0 0>}]
 dyn_bangs_B_03_r [{shapeType=Capsule, length=3, radius=1, center=<1 0 0>, rotation=<rot 90 0 0>}]
 dyn_bangs_C_01_l [{shapeType=Capsule, length=3, radius=1, center=<1 0 0>, rotation=<rot 90 0 0>}]
 dyn_bangs_C_01_r [{shapeType=Capsule, length=3, radius=1, center=<1 0 0>, rotation=<rot 90 0 0>}]
 dyn_bangs_C_02_l [{shapeType=Capsule, length=3, radius=1, center=<1 0 0>, rotation=<rot 90 0 0>}]
 dyn_bangs_C_02_r [{shapeType=Capsule, length=3, radius=1, center=<1 0 0>, rotation=<rot 90 0 0>}]
 dyn_hair_A_01_l [{shapeType=Capsule, length=3, radius=1, center=<1.5 0 0>, rotation=<rot 90 0 0>}]
 dyn_hair_A_01_r [{shapeType=Capsule, length=3, radius=1, center=<1 0 0>, rotation=<rot 90 0 0>}]
 dyn_hair_A_02_l [{shapeType=Capsule, length=3, radius=1, center=<1.5 0 0>, rotation=<rot 90 0 0>}]
 dyn_hair_A_02_r [{shapeType=Capsule, length=3, radius=1, center=<1 0 0>, rotation=<rot 90 0 0>}]
 dyn_hair_B_01_l [{shapeType=Capsule, length=3, radius=1, center=<1.5 0 0>, rotation=<rot 90 0 0>}]
 dyn_hair_B_01_r [{shapeType=Capsule, length=3, radius=1, center=<1 0 0>, rotation=<rot 90 0 0>}]
 dyn_hair_B_02_l [{shapeType=Capsule, length=3, radius=1, center=<1.5 0 0>, rotation=<rot 90 0 0>}]
 dyn_hair_B_02_r [{shapeType=Capsule, length=3, radius=1, center=<1 0 0>, rotation=<rot 90 0 0>}]
 dyn_hair_B_03_l [{shapeType=Capsule, length=3, radius=1, center=<1.5 0 0>, rotation=<rot 90 0 0>}]
 dyn_hair_B_03_r [{shapeType=Capsule, length=3, radius=1, center=<1 0 0>, rotation=<rot 90 0 0>}]
 dyn_hair_C_01_l [{shapeType=Capsule, length=3, radius=1, center=<1 0 0>, rotation=<rot 90 0 0>}]
 dyn_hair_C_01_r [{shapeType=Capsule, length=3, radius=1, center=<1 0 0>, rotation=<rot 90 0 0>}]
 dyn_hair_C_02_l [{shapeType=Capsule, length=3, radius=1, center=<1 0 0>, rotation=<rot 90 0 0>}]
 dyn_hair_C_02_r [{shapeType=Capsule, length=3, radius=1, center=<1 0 0>, rotation=<rot 90 0 0>}]
 dyn_hair_D_01_l [{shapeType=Capsule, length=3, radius=1, center=<1 0 0>, rotation=<rot 90 0 0>}]
 dyn_hair_D_01_r [{shapeType=Capsule, length=3, radius=1, center=<1 0 0>, rotation=<rot 90 0 0>}]
 dyn_hair_D_02_l [{shapeType=Capsule, length=3, radius=1, center=<1 0 0>, rotation=<rot 90 0 0>}]
 dyn_hair_D_02_r [{shapeType=Capsule, length=3, radius=1, center=<1 0 0>, rotation=<rot 90 0 0>}]
 dyn_hair_E_01_l [{shapeType=Capsule, length=3, radius=1, center=<2 0 0>, rotation=<rot 90 0 0>}]
 dyn_hair_E_01_r [{shapeType=Capsule, length=3, radius=1, center=<2 0 0>, rotation=<rot 90 0 0>}]
 dyn_hair_back_01 [{shapeType=Capsule, length=1.5, radius=2.5, center=<-0.0535 0 0>, rotation=<rot 90 0 0>}]
 dyn_hair_back_02 [{shapeType=Capsule, length=1.5, radius=2.5, center=<0.1 0 0>, rotation=<rot 90 0 0>}]
 dyn_hair_back_03 [{shapeType=Capsule, length=1.5, radius=2.5, center=<0.1 0 0>, rotation=<rot 90 0 0>}]
 head [{shapeType=Capsule, length=4.25, radius=7.5, center=<5.4515 4.5643 0>, rotation=<rot 90 -0.0099 -0.0099>}]
}
constraints [
 {bone1Name=dyn_bangs_A_01_l, bone2Name=head, swing1LimitDegrees=15, swing2LimitDegrees=15, twistLimitDegrees=1}
 {bone1Name=dyn_bangs_A_02_l, bone2Name=dyn_bangs_A_01_l, swing1LimitDegrees=15, swing2LimitDegrees=15, twistLimitDegrees=1}
 {bone1Name=dyn_bangs_A_03_l, bone2Name=dyn_bangs_A_02_l, swing1LimitDegrees=15, swing2LimitDegrees=15, twistLimitDegrees=1}
 {bone1Name=dyn_bangs_B_01_l, bone2Name=head, swing1LimitDegrees=15, swing2LimitDegrees=15, twistLimitDegrees=1}
 {bone1Name=dyn_bangs_B_02_l, bone2Name=dyn_bangs_B_01_l, swing1LimitDegrees=15, swing2LimitDegrees=15, twistLimitDegrees=1}
 {bone1Name=dyn_bangs_B_03_l, bone2Name=dyn_bangs_B_02_l, swing1LimitDegrees=15, swing2LimitDegrees=15, twistLimitDegrees=1}
 {bone1Name=dyn_bangs_C_01_l, bone2Name=head, swing1LimitDegrees=15, swing2LimitDegrees=15, twistLimitDegrees=1}
 {bone1Name=dyn_bangs_C_02_l, bone2Name=dyn_bangs_C_01_l, swing1LimitDegrees=15, swing2LimitDegrees=15, twistLimitDegrees=1}
 {bone1Name=dyn_hair_A_01_l, bone2Name=head, swing1LimitDegrees=15, swing2LimitDegrees=15, twistLimitDegrees=1}
 {bone1Name=dyn_hair_A_02_l, bone2Name=dyn_hair_A_01_l, swing1LimitDegrees=15, swing2LimitDegrees=15, twistLimitDegrees=1}
 {bone1Name=dyn_hair_B_01_l, bone2Name=head, swing1LimitDegrees=15, swing2LimitDegrees=15, twistLimitDegrees=1}
 {bone1Name=dyn_hair_B_02_l, bone2Name=dyn_hair_B_01_l, swing1LimitDegrees=15, swing2LimitDegrees=15, twistLimitDegrees=1}
 {bone1Name=dyn_hair_B_03_l, bone2Name=dyn_hair_B_02_l, swing1LimitDegrees=15, swing2LimitDegrees=15, twistLimitDegrees=1}
 {bone1Name=dyn_hair_E_01_l, bone2Name=head, swing1LimitDegrees=30, swing2LimitDegrees=30, twistLimitDegrees=1}
 {bone1Name=dyn_hair_D_01_l, bone2Name=head, swing1LimitDegrees=15, swing2LimitDegrees=15, twistLimitDegrees=1}
 {bone1Name=dyn_hair_D_02_l, bone2Name=dyn_hair_D_01_l, swing1LimitDegrees=15, swing2LimitDegrees=15, twistLimitDegrees=1}
 {bone1Name=dyn_hair_back_01, bone2Name=head, swing1LimitDegrees=45, swing2LimitDegrees=10, twistLimitDegrees=1}
 {bone1Name=dyn_hair_back_02, bone2Name=dyn_hair_back_01, swing1LimitDegrees=20, swing2LimitDegrees=20, twistLimitDegrees=1}
 {bone1Name=dyn_hair_back_03, bone2Name=dyn_hair_back_01, swing1LimitDegrees=20, swing2LimitDegrees=20, twistLimitDegrees=1}
 {bone1Name=dyn_hair_C_01_l, bone2Name=head, swing1LimitDegrees=15, swing2LimitDegrees=15, twistLimitDegrees=1}
 {bone1Name=dyn_hair_C_02_l, bone2Name=dyn_hair_C_01_l, swing1LimitDegrees=15, swing2LimitDegrees=15, twistLimitDegrees=1}
 {bone1Name=dyn_hair_C_01_r, bone2Name=head, swing1LimitDegrees=15, swing2LimitDegrees=15, twistLimitDegrees=1}
 {bone1Name=dyn_hair_C_02_r, bone2Name=dyn_hair_C_01_r, swing1LimitDegrees=15, swing2LimitDegrees=15, twistLimitDegrees=1}
 {bone1Name=dyn_hair_D_01_r, bone2Name=head, swing1LimitDegrees=15, swing2LimitDegrees=15, twistLimitDegrees=1}
 {bone1Name=dyn_hair_D_02_r, bone2Name=dyn_hair_D_01_r, swing1LimitDegrees=15, swing2LimitDegrees=15, twistLimitDegrees=1}
 {bone1Name=dyn_hair_E_01_r, bone2Name=head, swing1LimitDegrees=30, swing2LimitDegrees=30, twistLimitDegrees=1}
 {bone1Name=dyn_hair_B_01_r, bone2Name=head, swing1LimitDegrees=15, swing2LimitDegrees=15, twistLimitDegrees=1}
 {bone1Name=dyn_hair_B_02_r, bone2Name=dyn_hair_B_01_r, swing1LimitDegrees=15, swing2LimitDegrees=15, twistLimitDegrees=1}
 {bone1Name=dyn_hair_B_03_r, bone2Name=dyn_hair_B_02_r, swing1LimitDegrees=15, swing2LimitDegrees=15, twistLimitDegrees=1}
 {bone1Name=dyn_hair_A_01_r, bone2Name=head, swing1LimitDegrees=15, swing2LimitDegrees=15, twistLimitDegrees=1}
 {bone1Name=dyn_hair_A_02_r, bone2Name=dyn_hair_A_01_r, swing1LimitDegrees=15, swing2LimitDegrees=15, twistLimitDegrees=1}
 {bone1Name=dyn_bangs_C_01_r, bone2Name=head, swing1LimitDegrees=15, swing2LimitDegrees=15, twistLimitDegrees=1}
 {bone1Name=dyn_bangs_C_02_r, bone2Name=dyn_bangs_C_01_r, swing1LimitDegrees=15, swing2LimitDegrees=15, twistLimitDegrees=1}
 {bone1Name=dyn_bangs_B_01_r, bone2Name=head, swing1LimitDegrees=15, swing2LimitDegrees=15, twistLimitDegrees=1}
 {bone1Name=dyn_bangs_B_02_r, bone2Name=dyn_bangs_B_01_r, swing1LimitDegrees=15, swing2LimitDegrees=15, twistLimitDegrees=1}
 {bone1Name=dyn_bangs_B_03_r, bone2Name=dyn_bangs_B_02_r, swing1LimitDegrees=15, swing2LimitDegrees=15, twistLimitDegrees=1}
 {bone1Name=dyn_bangs_A_01_r, bone2Name=head, swing1LimitDegrees=15, swing2LimitDegrees=15, twistLimitDegrees=1}
 {bone1Name=dyn_bangs_A_02_r, bone2Name=dyn_bangs_A_01_r, swing1LimitDegrees=15, swing2LimitDegrees=15, twistLimitDegrees=1}
 {bone1Name=dyn_bangs_A_03_r, bone2Name=dyn_bangs_A_02_r, swing1LimitDegrees=15, swing2LimitDegrees=15, twistLimitDegrees=1}
]
