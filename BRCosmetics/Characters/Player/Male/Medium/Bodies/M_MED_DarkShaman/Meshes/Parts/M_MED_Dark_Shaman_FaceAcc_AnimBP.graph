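# Animation Blueprint for the male Dark Shaman face accessory mesh that adds secondary motion to hair and bangs.
# It copies the attached parent's pose, then runs a serial chain of single-bone AnimDynamics nodes in component space to output.
# Each uses root-relative angular springs, reduced gravity and angular limits; dyn_hair_back_01 blends at only 0.1 alpha.
# A RigidBody node and blank AnimDynamics node are unconnected, and the EventGraph has no logic.
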AnimBlueprint /BRCosmetics/Characters/Player/Male/Medium/Bodies/M_MED_DarkShaman/Meshes/Parts/M_MED_Dark_Shaman_FaceAcc_AnimBP
target_skeleton=&/M_MED_Dark_Shaman_FaceAcc_Skeleton
graphs [
 AnimGraph {nodes [
   Anim Dynamics(Alpha=1, ExternalForce="0,0,0") {
    bGravityOverrideInSimSpace=true
    alphaBoolBlend {blendOption=Linear}
    physicsBodyDefinitions [{
      sphereCollisionRadius=0
      boundBone {boneName=None}
      boxExtents=<1 1 1>
      constraintSetup {angularLimitsMax=<0 0 0>, angularLimitsMin=<0 0 0>, angularTarget=<0 0 0>}
      localJointOffset=<0 0 0>
     }]
    boundBone=~
   }
   RigidBody() {
    bClampLinearTranslationLimitToRefPose=true
    overridePhysicsAsset=&/M_MED_Dark_Shaman_FaceAcc_Physics
    simulationSpace=BaseBoneSpace
    alphaBoolBlend {blendOption=Linear}
    baseBoneRef {boneName=root}
    componentAppliedLinearAccClamp=<100 100 100>
    componentLinearAccScale=<0.5 0.5 2>
    componentLinearVelScale=<0.5 0.5 2>
   }
   n1 = Anim Dynamics - Bone: dyn_bangs_A_01_l(ComponentPose=Local To Component(LocalPose=Copy Pose From Mesh() {bUseAttachedParent=true}), Alpha=1, ExternalForce="0,0,0") {
    angularDampingOverride=1
    angularSpringConstant=1500
    bAngularSpring=true
    bGravityOverrideInSimSpace=true
    bOverrideAngularDamping=true
    gravityScale=0.35
    simulationSpace=RootRelative
    alphaBoolBlend {blendOption=Linear}
    boundBone {boneName=dyn_bangs_A_01_l}
    physicsBodyDefinitions [{
      sphereCollisionRadius=0
      boundBone {boneName=dyn_bangs_A_01_l}
      boxExtents=<5 10 10>
      constraintSetup {angularLimitsMax=<0 20 0>, angularLimitsMin=<0 0 -40>, angularTarget=<0 0 0>}
      localJointOffset=<5 0 0>
     }]
   }
   n2 = Anim Dynamics - Bone: dyn_bangs_A_02_l(ComponentPose=@n1, Alpha=1, ExternalForce="0,0,0") {
    angularDampingOverride=1
    angularSpringConstant=1500
    bAngularSpring=true
    bGravityOverrideInSimSpace=true
    bOverrideAngularDamping=true
    gravityScale=0.35
    simulationSpace=RootRelative
    alphaBoolBlend {blendOption=Linear}
    boundBone {boneName=dyn_bangs_A_02_l}
    physicsBodyDefinitions [{
      sphereCollisionRadius=0
      boundBone {boneName=dyn_bangs_A_02_l}
      boxExtents=<5 10 10>
      constraintSetup {angularLimitsMax=<0 20 0>, angularLimitsMin=<0 0 -40>, angularTarget=<0 0 0>}
      localJointOffset=<5 0 0>
     }]
   }
   n3 = Anim Dynamics - Bone: dyn_bangs_A_03_l(ComponentPose=@n2, Alpha=1, ExternalForce="0,0,0") {
    angularDampingOverride=1
    angularSpringConstant=1500
    bAngularSpring=true
    bGravityOverrideInSimSpace=true
    bOverrideAngularDamping=true
    gravityScale=0.35
    simulationSpace=RootRelative
    alphaBoolBlend {blendOption=Linear}
    boundBone {boneName=dyn_bangs_A_03_l}
    physicsBodyDefinitions [{
      sphereCollisionRadius=0
      boundBone {boneName=dyn_bangs_A_03_l}
      boxExtents=<5 10 10>
      constraintSetup {angularLimitsMax=<0 20 0>, angularLimitsMin=<0 0 -40>, angularTarget=<0 0 0>}
      localJointOffset=<5 0 0>
     }]
   }
   n4 = Anim Dynamics - Bone: dyn_bangs_B_01_l(ComponentPose=@n3, Alpha=1, ExternalForce="0,0,0") {
    angularDampingOverride=1
    angularSpringConstant=1500
    bAngularSpring=true
    bGravityOverrideInSimSpace=true
    bOverrideAngularDamping=true
    gravityScale=0.35
    simulationSpace=RootRelative
    alphaBoolBlend {blendOption=Linear}
    boundBone {boneName=dyn_bangs_B_01_l}
    physicsBodyDefinitions [{
      sphereCollisionRadius=0
      boundBone {boneName=dyn_bangs_B_01_l}
      boxExtents=<5 10 10>
      constraintSetup {angularLimitsMax=<0 20 0>, angularLimitsMin=<0 0 -40>, angularTarget=<0 0 0>}
      localJointOffset=<5 0 0>
     }]
   }
   n5 = Anim Dynamics - Bone: dyn_bangs_B_02_l(ComponentPose=@n4, Alpha=1, ExternalForce="0,0,0") {
    angularDampingOverride=1
    angularSpringConstant=1500
    bAngularSpring=true
    bGravityOverrideInSimSpace=true
    bOverrideAngularDamping=true
    gravityScale=0.35
    simulationSpace=RootRelative
    alphaBoolBlend {blendOption=Linear}
    boundBone {boneName=dyn_bangs_B_02_l}
    physicsBodyDefinitions [{
      sphereCollisionRadius=0
      boundBone {boneName=dyn_bangs_B_02_l}
      boxExtents=<5 10 10>
      constraintSetup {angularLimitsMax=<0 20 0>, angularLimitsMin=<0 0 -40>, angularTarget=<0 0 0>}
      localJointOffset=<5 0 0>
     }]
   }
   n6 = Anim Dynamics - Bone: dyn_bangs_B_03_l(ComponentPose=@n5, Alpha=1, ExternalForce="0,0,0") {
    angularDampingOverride=1
    angularSpringConstant=1500
    bAngularSpring=true
    bGravityOverrideInSimSpace=true
    bOverrideAngularDamping=true
    gravityScale=0.35
    simulationSpace=RootRelative
    alphaBoolBlend {blendOption=Linear}
    boundBone {boneName=dyn_bangs_B_03_l}
    physicsBodyDefinitions [{
      sphereCollisionRadius=0
      boundBone {boneName=dyn_bangs_B_03_l}
      boxExtents=<5 10 10>
      constraintSetup {angularLimitsMax=<0 20 0>, angularLimitsMin=<0 0 -40>, angularTarget=<0 0 0>}
      localJointOffset=<5 0 0>
     }]
   }
   n7 = Anim Dynamics - Bone: dyn_bangs_C_01_l(ComponentPose=@n6, Alpha=1, ExternalForce="0,0,0") {
    angularDampingOverride=1
    angularSpringConstant=1500
    bAngularSpring=true
    bGravityOverrideInSimSpace=true
    bOverrideAngularDamping=true
    gravityScale=0.35
    simulationSpace=RootRelative
    alphaBoolBlend {blendOption=Linear}
    boundBone {boneName=dyn_bangs_C_01_l}
    physicsBodyDefinitions [{
      sphereCollisionRadius=0
      boundBone {boneName=dyn_bangs_C_01_l}
      boxExtents=<5 10 10>
      constraintSetup {angularLimitsMax=<0 20 0>, angularLimitsMin=<0 0 -40>, angularTarget=<0 0 0>}
      localJointOffset=<5 0 0>
     }]
   }
   n8 = Anim Dynamics - Bone: dyn_bangs_C_02_l(ComponentPose=@n7, Alpha=1, ExternalForce="0,0,0") {
    angularDampingOverride=1
    angularSpringConstant=1500
    bAngularSpring=true
    bGravityOverrideInSimSpace=true
    bOverrideAngularDamping=true
    gravityScale=0.35
    simulationSpace=RootRelative
    alphaBoolBlend {blendOption=Linear}
    boundBone {boneName=dyn_bangs_C_02_l}
    physicsBodyDefinitions [{
      sphereCollisionRadius=0
      boundBone {boneName=dyn_bangs_C_02_l}
      boxExtents=<5 10 10>
      constraintSetup {angularLimitsMax=<0 20 0>, angularLimitsMin=<0 0 -40>, angularTarget=<0 0 0>}
      localJointOffset=<5 0 0>
     }]
   }
   n9 = Anim Dynamics - Bone: dyn_bangs_C_01_r(ComponentPose=@n8, Alpha=1, ExternalForce="0,0,0") {
    angularDampingOverride=1
    angularSpringConstant=1500
    bAngularSpring=true
    bGravityOverrideInSimSpace=true
    bOverrideAngularDamping=true
    gravityScale=0.35
    simulationSpace=RootRelative
    alphaBoolBlend {blendOption=Linear}
    boundBone {boneName=dyn_bangs_C_01_r}
    physicsBodyDefinitions [{
      sphereCollisionRadius=0
      boundBone {boneName=dyn_bangs_C_01_r}
      boxExtents=<5 10 10>
      constraintSetup {angularLimitsMax=<0 20 40>, angularLimitsMin=<0 0 -10>, angularTarget=<0 0 0>}
      localJointOffset=<5 0 0>
     }]
   }
   n10 = Anim Dynamics - Bone: dyn_bangs_C_02_r(ComponentPose=@n9, Alpha=1, ExternalForce="0,0,0") {
    angularDampingOverride=1
    angularSpringConstant=1500
    bAngularSpring=true
    bGravityOverrideInSimSpace=true
    bOverrideAngularDamping=true
    gravityScale=0.35
    simulationSpace=RootRelative
    alphaBoolBlend {blendOption=Linear}
    boundBone {boneName=dyn_bangs_C_02_r}
    physicsBodyDefinitions [{
      sphereCollisionRadius=0
      boundBone {boneName=dyn_bangs_C_02_r}
      boxExtents=<5 10 10>
      constraintSetup {angularLimitsMax=<0 20 40>, angularLimitsMin=<0 0 -10>, angularTarget=<0 0 0>}
      localJointOffset=<5 0 0>
     }]
   }
   n11 = Anim Dynamics - Bone: dyn_bangs_B_01_r(ComponentPose=@n10, Alpha=1, ExternalForce="0,0,0") {
    angularDampingOverride=1
    angularSpringConstant=1500
    bAngularSpring=true
    bGravityOverrideInSimSpace=true
    bOverrideAngularDamping=true
    gravityScale=0.35
    simulationSpace=RootRelative
    alphaBoolBlend {blendOption=Linear}
    boundBone {boneName=dyn_bangs_B_01_r}
    physicsBodyDefinitions [{
      sphereCollisionRadius=0
      boundBone {boneName=dyn_bangs_B_01_r}
      boxExtents=<5 10 10>
      constraintSetup {angularLimitsMax=<0 20 40>, angularLimitsMin=<0 0 -10>, angularTarget=<0 0 0>}
      localJointOffset=<5 0 0>
     }]
   }
   n12 = Anim Dynamics - Bone: dyn_bangs_B_02_r(ComponentPose=@n11, Alpha=1, ExternalForce="0,0,0") {
    angularDampingOverride=1
    angularSpringConstant=1500
    bAngularSpring=true
    bGravityOverrideInSimSpace=true
    bOverrideAngularDamping=true
    gravityScale=0.35
    simulationSpace=RootRelative
    alphaBoolBlend {blendOption=Linear}
    boundBone {boneName=dyn_bangs_B_02_r}
    physicsBodyDefinitions [{
      sphereCollisionRadius=0
      boundBone {boneName=dyn_bangs_B_02_r}
      boxExtents=<5 10 10>
      constraintSetup {angularLimitsMax=<0 20 40>, angularLimitsMin=<0 0 -10>, angularTarget=<0 0 0>}
      localJointOffset=<5 0 0>
     }]
   }
   n13 = Anim Dynamics - Bone: dyn_bangs_B_03_r(ComponentPose=@n12, Alpha=1, ExternalForce="0,0,0") {
    angularDampingOverride=1
    angularSpringConstant=1500
    bAngularSpring=true
    bGravityOverrideInSimSpace=true
    bOverrideAngularDamping=true
    gravityScale=0.35
    simulationSpace=RootRelative
    alphaBoolBlend {blendOption=Linear}
    boundBone {boneName=dyn_bangs_B_03_r}
    physicsBodyDefinitions [{
      sphereCollisionRadius=0
      boundBone {boneName=dyn_bangs_B_03_r}
      boxExtents=<5 10 10>
      constraintSetup {angularLimitsMax=<0 20 40>, angularLimitsMin=<0 0 -10>, angularTarget=<0 0 0>}
      localJointOffset=<5 0 0>
     }]
   }
   n14 = Anim Dynamics - Bone: dyn_bangs_A_01_r(ComponentPose=@n13, Alpha=1, ExternalForce="0,0,0") {
    angularDampingOverride=1
    angularSpringConstant=1500
    bAngularSpring=true
    bGravityOverrideInSimSpace=true
    bOverrideAngularDamping=true
    gravityScale=0.35
    simulationSpace=RootRelative
    alphaBoolBlend {blendOption=Linear}
    boundBone {boneName=dyn_bangs_A_01_r}
    physicsBodyDefinitions [{
      sphereCollisionRadius=0
      boundBone {boneName=dyn_bangs_A_01_r}
      boxExtents=<5 10 10>
      constraintSetup {angularLimitsMax=<0 20 40>, angularLimitsMin=<0 0 -10>, angularTarget=<0 0 0>}
      localJointOffset=<5 0 0>
     }]
   }
   n15 = Anim Dynamics - Bone: dyn_bangs_A_02_r(ComponentPose=@n14, Alpha=1, ExternalForce="0,0,0") {
    angularDampingOverride=1
    angularSpringConstant=1500
    bAngularSpring=true
    bGravityOverrideInSimSpace=true
    bOverrideAngularDamping=true
    gravityScale=0.35
    simulationSpace=RootRelative
    alphaBoolBlend {blendOption=Linear}
    boundBone {boneName=dyn_bangs_A_02_r}
    physicsBodyDefinitions [{
      sphereCollisionRadius=0
      boundBone {boneName=dyn_bangs_A_02_r}
      boxExtents=<5 10 10>
      constraintSetup {angularLimitsMax=<0 20 40>, angularLimitsMin=<0 0 -10>, angularTarget=<0 0 0>}
      localJointOffset=<5 0 0>
     }]
   }
   n16 = Anim Dynamics - Bone: dyn_bangs_A_03_r(ComponentPose=@n15, Alpha=1, ExternalForce="0,0,0") {
    angularDampingOverride=1
    angularSpringConstant=1500
    bAngularSpring=true
    bGravityOverrideInSimSpace=true
    bOverrideAngularDamping=true
    gravityScale=0.35
    simulationSpace=RootRelative
    alphaBoolBlend {blendOption=Linear}
    boundBone {boneName=dyn_bangs_A_03_r}
    physicsBodyDefinitions [{
      sphereCollisionRadius=0
      boundBone {boneName=dyn_bangs_A_03_r}
      boxExtents=<5 10 10>
      constraintSetup {angularLimitsMax=<0 20 40>, angularLimitsMin=<0 0 -10>, angularTarget=<0 0 0>}
      localJointOffset=<5 0 0>
     }]
   }
   n17 = Anim Dynamics - Bone: dyn_hair_A_01_l(ComponentPose=@n16, Alpha=1, ExternalForce="0,0,0") {
    angularDampingOverride=1
    angularSpringConstant=1500
    bAngularSpring=true
    bGravityOverrideInSimSpace=true
    bOverrideAngularDamping=true
    gravityScale=0.35
    simulationSpace=RootRelative
    alphaBoolBlend {blendOption=Linear}
    boundBone {boneName=dyn_hair_A_01_l}
    physicsBodyDefinitions [{
      sphereCollisionRadius=0
      boundBone {boneName=dyn_hair_A_01_l}
      boxExtents=<5 10 10>
      constraintSetup {angularLimitsMax=<0 20 0>, angularLimitsMin=<0 0 -40>, angularTarget=<0 0 0>}
      localJointOffset=<5 0 0>
     }]
   }
   n18 = Anim Dynamics - Bone: dyn_hair_A_02_l(ComponentPose=@n17, Alpha=1, ExternalForce="0,0,0") {
    angularDampingOverride=1
    angularSpringConstant=1500
    bAngularSpring=true
    bGravityOverrideInSimSpace=true
    bOverrideAngularDamping=true
    gravityScale=0.35
    simulationSpace=RootRelative
    alphaBoolBlend {blendOption=Linear}
    boundBone {boneName=dyn_hair_A_02_l}
    physicsBodyDefinitions [{
      sphereCollisionRadius=0
      boundBone {boneName=dyn_hair_A_02_l}
      boxExtents=<5 10 10>
      constraintSetup {angularLimitsMax=<0 20 0>, angularLimitsMin=<0 0 -40>, angularTarget=<0 0 0>}
      localJointOffset=<5 0 0>
     }]
   }
   n19 = Anim Dynamics - Bone: dyn_hair_B_01_l(ComponentPose=@n18, Alpha=1, ExternalForce="0,0,0") {
    angularDampingOverride=1
    angularSpringConstant=1500
    bAngularSpring=true
    bGravityOverrideInSimSpace=true
    bOverrideAngularDamping=true
    gravityScale=0.35
    simulationSpace=RootRelative
    alphaBoolBlend {blendOption=Linear}
    boundBone {boneName=dyn_hair_B_01_l}
    physicsBodyDefinitions [{
      sphereCollisionRadius=0
      boundBone {boneName=dyn_hair_B_01_l}
      boxExtents=<5 10 10>
      constraintSetup {angularLimitsMax=<0 20 0>, angularLimitsMin=<0 0 -40>, angularTarget=<0 0 0>}
      localJointOffset=<5 0 0>
     }]
   }
   n20 = Anim Dynamics - Bone: dyn_hair_B_02_l(ComponentPose=@n19, Alpha=1, ExternalForce="0,0,0") {
    angularDampingOverride=1
    angularSpringConstant=1500
    bAngularSpring=true
    bGravityOverrideInSimSpace=true
    bOverrideAngularDamping=true
    gravityScale=0.35
    simulationSpace=RootRelative
    alphaBoolBlend {blendOption=Linear}
    boundBone {boneName=dyn_hair_B_02_l}
    physicsBodyDefinitions [{
      sphereCollisionRadius=0
      boundBone {boneName=dyn_hair_B_02_l}
      boxExtents=<5 10 10>
      constraintSetup {angularLimitsMax=<0 20 0>, angularLimitsMin=<0 0 -40>, angularTarget=<0 0 0>}
      localJointOffset=<5 0 0>
     }]
   }
   n21 = Anim Dynamics - Bone: dyn_hair_B_03_l(ComponentPose=@n20, Alpha=1, ExternalForce="0,0,0") {
    angularDampingOverride=1
    angularSpringConstant=1500
    bAngularSpring=true
    bGravityOverrideInSimSpace=true
    bOverrideAngularDamping=true
    gravityScale=0.35
    simulationSpace=RootRelative
    alphaBoolBlend {blendOption=Linear}
    boundBone {boneName=dyn_hair_B_03_l}
    physicsBodyDefinitions [{
      sphereCollisionRadius=0
      boundBone {boneName=dyn_hair_B_03_l}
      boxExtents=<5 10 10>
      constraintSetup {angularLimitsMax=<0 20 0>, angularLimitsMin=<0 0 -40>, angularTarget=<0 0 0>}
      localJointOffset=<5 0 0>
     }]
   }
   n22 = Anim Dynamics - Bone: dyn_hair_D_01_l(ComponentPose=@n21, Alpha=1, ExternalForce="0,0,0") {
    angularDampingOverride=1
    angularSpringConstant=1500
    bAngularSpring=true
    bGravityOverrideInSimSpace=true
    bOverrideAngularDamping=true
    gravityScale=0.35
    simulationSpace=RootRelative
    alphaBoolBlend {blendOption=Linear}
    boundBone {boneName=dyn_hair_D_01_l}
    physicsBodyDefinitions [{
      sphereCollisionRadius=0
      boundBone {boneName=dyn_hair_D_01_l}
      boxExtents=<5 10 10>
      constraintSetup {angularLimitsMax=<0 20 0>, angularLimitsMin=<0 0 -40>, angularTarget=<0 0 0>}
      localJointOffset=<5 0 0>
     }]
   }
   n23 = Anim Dynamics - Bone: dyn_hair_D_02_l(ComponentPose=@n22, Alpha=1, ExternalForce="0,0,0") {
    angularDampingOverride=1
    angularSpringConstant=1500
    bAngularSpring=true
    bGravityOverrideInSimSpace=true
    bOverrideAngularDamping=true
    gravityScale=0.35
    simulationSpace=RootRelative
    alphaBoolBlend {blendOption=Linear}
    boundBone {boneName=dyn_hair_D_02_l}
    physicsBodyDefinitions [{
      sphereCollisionRadius=0
      boundBone {boneName=dyn_hair_D_02_l}
      boxExtents=<5 10 10>
      constraintSetup {angularLimitsMax=<0 20 0>, angularLimitsMin=<0 0 -40>, angularTarget=<0 0 0>}
      localJointOffset=<5 0 0>
     }]
   }
   n24 = Anim Dynamics - Bone: dyn_hair_C_01_l(ComponentPose=@n23, Alpha=1, ExternalForce="0,0,0") {
    angularDampingOverride=1
    angularSpringConstant=1500
    bAngularSpring=true
    bGravityOverrideInSimSpace=true
    bOverrideAngularDamping=true
    gravityScale=0.35
    simulationSpace=RootRelative
    alphaBoolBlend {blendOption=Linear}
    boundBone {boneName=dyn_hair_C_01_l}
    physicsBodyDefinitions [{
      sphereCollisionRadius=0
      boundBone {boneName=dyn_hair_C_01_l}
      boxExtents=<5 10 10>
      constraintSetup {angularLimitsMax=<0 20 0>, angularLimitsMin=<0 0 -40>, angularTarget=<0 0 0>}
      localJointOffset=<5 0 0>
     }]
   }
   n25 = Anim Dynamics - Bone: dyn_hair_C_02_l(ComponentPose=@n24, Alpha=1, ExternalForce="0,0,0") {
    angularDampingOverride=1
    angularSpringConstant=1500
    bAngularSpring=true
    bGravityOverrideInSimSpace=true
    bOverrideAngularDamping=true
    gravityScale=0.35
    simulationSpace=RootRelative
    alphaBoolBlend {blendOption=Linear}
    boundBone {boneName=dyn_hair_C_02_l}
    physicsBodyDefinitions [{
      sphereCollisionRadius=0
      boundBone {boneName=dyn_hair_C_02_l}
      boxExtents=<5 10 10>
      constraintSetup {angularLimitsMax=<0 20 0>, angularLimitsMin=<0 0 -40>, angularTarget=<0 0 0>}
      localJointOffset=<5 0 0>
     }]
   }
   n26 = Anim Dynamics - Bone: dyn_hair_C_01_r(ComponentPose=@n25, Alpha=1, ExternalForce="0,0,0") {
    angularDampingOverride=1
    angularSpringConstant=1500
    bAngularSpring=true
    bGravityOverrideInSimSpace=true
    bOverrideAngularDamping=true
    gravityScale=0.35
    simulationSpace=RootRelative
    alphaBoolBlend {blendOption=Linear}
    boundBone {boneName=dyn_hair_C_01_r}
    physicsBodyDefinitions [{
      sphereCollisionRadius=0
      boundBone {boneName=dyn_hair_C_01_r}
      boxExtents=<5 10 10>
      constraintSetup {angularLimitsMax=<0 20 0>, angularLimitsMin=<0 0 -40>, angularTarget=<0 0 0>}
      localJointOffset=<5 0 0>
     }]
   }
   n27 = Anim Dynamics - Bone: dyn_hair_C_02_r(ComponentPose=@n26, Alpha=1, ExternalForce="0,0,0") {
    angularDampingOverride=1
    angularSpringConstant=1500
    bAngularSpring=true
    bGravityOverrideInSimSpace=true
    bOverrideAngularDamping=true
    gravityScale=0.35
    simulationSpace=RootRelative
    alphaBoolBlend {blendOption=Linear}
    boundBone {boneName=dyn_hair_C_02_r}
    physicsBodyDefinitions [{
      sphereCollisionRadius=0
      boundBone {boneName=dyn_hair_C_02_r}
      boxExtents=<5 10 10>
      constraintSetup {angularLimitsMax=<0 20 0>, angularLimitsMin=<0 0 -40>, angularTarget=<0 0 0>}
      localJointOffset=<5 0 0>
     }]
   }
   n28 = Anim Dynamics - Bone: dyn_hair_D_01_r(ComponentPose=@n27, Alpha=1, ExternalForce="0,0,0") {
    angularDampingOverride=1
    angularSpringConstant=1500
    bAngularSpring=true
    bGravityOverrideInSimSpace=true
    bOverrideAngularDamping=true
    gravityScale=0.35
    simulationSpace=RootRelative
    alphaBoolBlend {blendOption=Linear}
    boundBone {boneName=dyn_hair_D_01_r}
    physicsBodyDefinitions [{
      sphereCollisionRadius=0
      boundBone {boneName=dyn_hair_D_01_r}
      boxExtents=<5 10 10>
      constraintSetup {angularLimitsMax=<0 20 0>, angularLimitsMin=<0 0 -40>, angularTarget=<0 0 0>}
      localJointOffset=<5 0 0>
     }]
   }
   n29 = Anim Dynamics - Bone: dyn_hair_D_02_r(ComponentPose=@n28, Alpha=1, ExternalForce="0,0,0") {
    angularDampingOverride=1
    angularSpringConstant=1500
    bAngularSpring=true
    bGravityOverrideInSimSpace=true
    bOverrideAngularDamping=true
    gravityScale=0.35
    simulationSpace=RootRelative
    alphaBoolBlend {blendOption=Linear}
    boundBone {boneName=dyn_hair_D_02_r}
    physicsBodyDefinitions [{
      sphereCollisionRadius=0
      boundBone {boneName=dyn_hair_D_02_r}
      boxExtents=<5 10 10>
      constraintSetup {angularLimitsMax=<0 20 0>, angularLimitsMin=<0 0 -40>, angularTarget=<0 0 0>}
      localJointOffset=<5 0 0>
     }]
   }
   n30 = Anim Dynamics - Bone: dyn_hair_B_01_r(ComponentPose=@n29, Alpha=1, ExternalForce="0,0,0") {
    angularDampingOverride=1
    angularSpringConstant=1500
    bAngularSpring=true
    bGravityOverrideInSimSpace=true
    bOverrideAngularDamping=true
    gravityScale=0.35
    simulationSpace=RootRelative
    alphaBoolBlend {blendOption=Linear}
    boundBone {boneName=dyn_hair_B_01_r}
    physicsBodyDefinitions [{
      sphereCollisionRadius=0
      boundBone {boneName=dyn_hair_B_01_r}
      boxExtents=<5 10 10>
      constraintSetup {angularLimitsMax=<0 20 0>, angularLimitsMin=<0 0 -40>, angularTarget=<0 0 0>}
      localJointOffset=<5 0 0>
     }]
   }
   n31 = Anim Dynamics - Bone: dyn_hair_B_02_r(ComponentPose=@n30, Alpha=1, ExternalForce="0,0,0") {
    angularDampingOverride=1
    angularSpringConstant=1500
    bAngularSpring=true
    bGravityOverrideInSimSpace=true
    bOverrideAngularDamping=true
    gravityScale=0.35
    simulationSpace=RootRelative
    alphaBoolBlend {blendOption=Linear}
    boundBone {boneName=dyn_hair_B_02_r}
    physicsBodyDefinitions [{
      sphereCollisionRadius=0
      boundBone {boneName=dyn_hair_B_02_r}
      boxExtents=<5 10 10>
      constraintSetup {angularLimitsMax=<0 20 0>, angularLimitsMin=<0 0 -40>, angularTarget=<0 0 0>}
      localJointOffset=<5 0 0>
     }]
   }
   n32 = Anim Dynamics - Bone: dyn_hair_B_03_r(ComponentPose=@n31, Alpha=1, ExternalForce="0,0,0") {
    angularDampingOverride=1
    angularSpringConstant=1500
    bAngularSpring=true
    bGravityOverrideInSimSpace=true
    bOverrideAngularDamping=true
    gravityScale=0.35
    simulationSpace=RootRelative
    alphaBoolBlend {blendOption=Linear}
    boundBone {boneName=dyn_hair_B_03_r}
    physicsBodyDefinitions [{
      sphereCollisionRadius=0
      boundBone {boneName=dyn_hair_B_03_r}
      boxExtents=<5 10 10>
      constraintSetup {angularLimitsMax=<0 20 0>, angularLimitsMin=<0 0 -40>, angularTarget=<0 0 0>}
      localJointOffset=<5 0 0>
     }]
   }
   n33 = Anim Dynamics - Bone: dyn_hair_A_01_r(ComponentPose=@n32, Alpha=1, ExternalForce="0,0,0") {
    angularDampingOverride=1
    angularSpringConstant=1500
    bAngularSpring=true
    bGravityOverrideInSimSpace=true
    bOverrideAngularDamping=true
    gravityScale=0.35
    simulationSpace=RootRelative
    alphaBoolBlend {blendOption=Linear}
    boundBone {boneName=dyn_hair_A_01_r}
    physicsBodyDefinitions [{
      sphereCollisionRadius=0
      boundBone {boneName=dyn_hair_A_01_r}
      boxExtents=<5 10 10>
      constraintSetup {angularLimitsMax=<0 20 0>, angularLimitsMin=<0 0 -40>, angularTarget=<0 0 0>}
      localJointOffset=<5 0 0>
     }]
   }
   n34 = Anim Dynamics - Bone: dyn_hair_A_02_r(ComponentPose=@n33, Alpha=1, ExternalForce="0,0,0") {
    angularDampingOverride=1
    angularSpringConstant=1500
    bAngularSpring=true
    bGravityOverrideInSimSpace=true
    bOverrideAngularDamping=true
    gravityScale=0.35
    simulationSpace=RootRelative
    alphaBoolBlend {blendOption=Linear}
    boundBone {boneName=dyn_hair_A_02_r}
    physicsBodyDefinitions [{
      sphereCollisionRadius=0
      boundBone {boneName=dyn_hair_A_02_r}
      boxExtents=<5 10 10>
      constraintSetup {angularLimitsMax=<0 20 0>, angularLimitsMin=<0 0 -40>, angularTarget=<0 0 0>}
      localJointOffset=<5 0 0>
     }]
   }
   n35 = Anim Dynamics - Bone: dyn_hair_E_01_r(ComponentPose=@n34, Alpha=1, ExternalForce="0,0,0") {
    angularDampingOverride=1
    angularSpringConstant=1500
    bAngularSpring=true
    bGravityOverrideInSimSpace=true
    bOverrideAngularDamping=true
    gravityScale=0.35
    simulationSpace=RootRelative
    alphaBoolBlend {blendOption=Linear}
    boundBone {boneName=dyn_hair_E_01_r}
    physicsBodyDefinitions [{
      sphereCollisionRadius=0
      boundBone {boneName=dyn_hair_E_01_r}
      boxExtents=<5 10 10>
      constraintSetup {angularLimitsMax=<0 20 20>, angularLimitsMin=<0 -20 -40>, angularTarget=<0 0 0>}
      localJointOffset=<5 2 2>
     }]
   }
   n36 = Anim Dynamics - Bone: dyn_hair_E_01_l(ComponentPose=@n35, Alpha=1, ExternalForce="0,0,0") {
    angularDampingOverride=1
    angularSpringConstant=1500
    bAngularSpring=true
    bGravityOverrideInSimSpace=true
    bOverrideAngularDamping=true
    gravityScale=0.35
    simulationSpace=RootRelative
    alphaBoolBlend {blendOption=Linear}
    boundBone {boneName=dyn_hair_E_01_l}
    physicsBodyDefinitions [{
      sphereCollisionRadius=0
      boundBone {boneName=dyn_hair_E_01_l}
      boxExtents=<5 12 8>
      constraintSetup {angularLimitsMax=<0 40 40>, angularLimitsMin=<0 -40 -40>, angularTarget=<0 0 0>}
      localJointOffset=<5 -1 0>
     }]
   }
   n37 = Anim Dynamics - Bone: dyn_hair_back_01(ComponentPose=@n36, Alpha=0.1, ExternalForce="0,0,0") {
    angularDampingOverride=1
    angularSpringConstant=1500
    bAngularSpring=true
    bGravityOverrideInSimSpace=true
    bOverrideAngularDamping=true
    gravityScale=0.35
    simulationSpace=RootRelative
    alphaBoolBlend {blendOption=Linear}
    boundBone {boneName=dyn_hair_back_01}
    physicsBodyDefinitions [{
      sphereCollisionRadius=0
      boundBone {boneName=dyn_hair_back_01}
      boxExtents=<5 12 8>
      constraintSetup {angularLimitsMax=<0 20 40>, angularLimitsMin=<0 -20 -20>, angularTarget=<0 0 0>}
      localJointOffset=<0 0 5>
     }]
   }
   n38 = Anim Dynamics - Bone: dyn_hair_back_03(ComponentPose=@n37, Alpha=1, ExternalForce="0,0,0") {
    angularDampingOverride=1
    angularSpringConstant=1500
    bAngularSpring=true
    bGravityOverrideInSimSpace=true
    bOverrideAngularDamping=true
    gravityScale=0.35
    simulationSpace=RootRelative
    alphaBoolBlend {blendOption=Linear}
    boundBone {boneName=dyn_hair_back_03}
    physicsBodyDefinitions [{
      sphereCollisionRadius=0
      boundBone {boneName=dyn_hair_back_03}
      boxExtents=<5 12 8>
      constraintSetup {angularLimitsMax=<0 20 40>, angularLimitsMin=<0 -20 -20>, angularTarget=<0 0 0>}
      localJointOffset=<0 0 5>
     }]
   }
   n39 = Anim Dynamics - Bone: dyn_hair_back_02(ComponentPose=@n38, Alpha=1, ExternalForce="0,0,0") {
    angularDampingOverride=1
    angularSpringConstant=1500
    bAngularSpring=true
    bGravityOverrideInSimSpace=true
    bOverrideAngularDamping=true
    gravityScale=0.35
    simulationSpace=RootRelative
    alphaBoolBlend {blendOption=Linear}
    boundBone {boneName=dyn_hair_back_02}
    physicsBodyDefinitions [{
      sphereCollisionRadius=0
      boundBone {boneName=dyn_hair_back_02}
      boxExtents=<5 12 8>
      constraintSetup {angularLimitsMax=<0 40 40>, angularLimitsMin=<0 -20 -20>, angularTarget=<0 0 0>}
      localJointOffset=<0 0 5>
     }]
   }
   Output Pose(Result=Component To Local(ComponentPose=@n39)) {layerGroup=None}
  ]}
 EventGraph {nodes [
   TryGetPawnOwner()
   Event BlueprintUpdateAnimation()
  ]}
]
variables []
